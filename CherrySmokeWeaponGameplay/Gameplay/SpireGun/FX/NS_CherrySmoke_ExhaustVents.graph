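# Niagara effect for the CherrySmoke SpireGun's exhaust vents, placing sprites on ten indexed skeletal-mesh bone/socket locations in left/right pairs.
# VentSmoke emits a tapering smoke stream, using age-driven partition masks and shifts to switch pairs on, then off, in sequence.
# It computes separate stream and puff attributes for two sprite renderers, with time-of-day tinting.
# VentSmoke001 bursts longer-lived, fading hot glows at the vents; Minimal001 is disabled.
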NiagaraSystem /CherrySmokeWeaponGameplay/Gameplay/SpireGun/FX/NS_CherrySmoke_ExhaustVents
dependencies {
 usedDataInterfaces [
  &NiagaraDataInterfaceCurve
  &NiagaraDataInterfaceSpriteRendererInfo
  &NiagaraDataInterfaceColorCurve
  &NiagaraDataInterfaceSkeletalMesh
 ]
 usedDynamicInputs []
 usedModules [
  &self:SystemSpawnScript.SystemScriptSource.SystemScriptGraph.NiagaraNodeAssignment_0.SetVariables_#1
  &SystemState
 ]
}
emitters {
 Minimal001 {
  bEnabled=false
  emitterUpdateScript [
   &EmitterState(Life Cycle Mode=::NewEnumerator1, Loop Behavior=::NewEnumerator1, Loop Duration=0.3, UseLoopDelay=false)
   &SpawnRate(SpawnRate=50, Use Spawn Probability=false)
   &SpawnBurst_Instantaneous(Spawn Count=10, Use Spawn Probability=false, UseLoopCountLimit=false) {enabled=false}
  ]
  particleSpawnScript [
   &InitializeParticle(Sprite Size Mode=::NewEnumerator3, Uniform Sprite Size=15, UsePositionOffset=false, Write Material Random=true)
   &PartitionParticles(Number Of Partitions=&MakeIntFromFloat(Float=&FloatFromCurve(
      FloatCurve=DataInterface() {propertyValues {Curve {keys [<at 0 2 RCIM_Constant>,<at 0.05 4 RCIM_Constant>,<at 0.1 6 RCIM_Constant>,<at 0.15 8 RCIM_Constant>,<at 0.2 10 RCIM_Constant>]}}}
      CurveIndex=$Emitter.Age
     )))
   &CameraOffset(Camera Offset Amount=-1) {enabled=false}
   &SpriteFacingAndAlignment(
    Sprite Alignment=&RotateVectorByQuaternion(Vector To Rotate=<0 0 1>, Quaternion=$Particles.SkeletalMeshLocation.SampledRotation)
    WriteAlignment=true
    WriteFacing=true
   ) {enabled=false}
   &SkeletalMeshLocation(
    Bone / Socket Index=$Particles.PartitionParticles.Partition
    Bone Sampling Mode=::NewEnumerator4
    MaskSpawnGroup=false
    Sampled Position Offset=&MakeVector(
     X=-1.6
     Y=&MergeHelper_FloatSwitch(0=-2, 1=2, 2=?, 3=?, Index=&ModuloInt(A=$Particles.PartitionParticles.Partition, B=2), Num=::NewEnumerator0)
     Z=-3
    )
    Skeletal Mesh=$System.Skeletal Mesh
   ) {enabled=false, module=SkeletalMeshLocation001}
   &AddVelocity(
    Velocity=&MakeVector(
     X=-1
     Y=&MergeHelper_FloatSwitch(0=-2, 1=2, 2=?, 3=?, Index=&ModuloInt(A=$Particles.PartitionParticles.Partition, B=2), Num=::NewEnumerator0)
     Z=1
    )
    Velocity Speed Scale=50
   )
   &AddVelocity(Velocity=<0 0 1>) {enabled=false, module=AddVelocity001}
  ]
  particleUpdateScript [
   &ParticleState()
   &SkeletalMeshLocation(
    Bone / Socket Index=$Particles.PartitionParticles.Partition
    Bone Sampling Mode=::NewEnumerator4
    MaskSpawnGroup=false
    Position Sampling=::NewEnumerator1
    Sampled Position Offset Coordinate Space:ENiagaraSkelMeshTransforms=::NewEnumerator2
    Skeletal Mesh=$System.Skeletal Mesh
    Source Movement Scale:Vector3f=&VectorFromFloat(Value=&FloatFromCurve(FloatCurve=DataInterface() {propertyValues {Curve {keys [
          {interpMode=RCIM_Cubic, tangentMode=RCTM_User, time=0, value=1}
          {arriveTangent=0.0173, interpMode=RCIM_Cubic, leaveTangent=0.0173, tangentMode=RCTM_User, time=1, value=0.5}
         ]}}}))
   )
   &ScaleSpriteSize(
    Scale Factor=&MakeVector2D(
     X=1
     Y=&FloatFromCurve(
      FloatCurve=DataInterface() {propertyValues {Curve {keys [<at 0 0.3>,<at 1 1>]}}}
      Scale Curve=&Lerp_Float(
       A=17
       B=5
       Alpha=&FloatFromCurve(FloatCurve=DataInterface() {propertyValues {Curve {keys [<at 0 0>,<at 0.5 1>]}}}, CurveIndex=$Emitter.Age)
      )
     )
    )
    Scale Sprite Size Mode=::NewEnumerator2
   ) {enabled=false}
   &Drag(Drag=10, Use Linear Drag=true, Use Rotational Drag=true)
   &InheritSourceMovement(
    Applied Position Scale=&VectorFromFloat(Value=1)
    ApplyForce=false
    ApplyPosition=true
    ApplyPositionFalloff=false
   ) {enabled=false}
   &SolveForcesAndVelocity(Clamp Velocity=false, Limit Acceleration=false)
   &SubUVAnimation(
    "Sprite Renderer "=DataInterface() {propertyValues {SpriteRenderer=&self:Minimal001_0.NiagaraSpriteRendererProperties_0}}
    UseEndFrame=false
    UseStartFrame=false
   )
   &ScaleColor(
    Scale Alpha=&FloatFromCurve(FloatCurve=DataInterface() {propertyValues {Curve {keys [<at 0 1>,<at 1 0>]}}}, Scale Curve=0.2)
    Scale RGB=&MakeVectorFromLinearColorRGB(Color=&ColorFromCurve(ColorCurve=DataInterface() {propertyValues {
        AlphaCurve {keys [<at 0 1 RCIM_Cubic>]}
        BlueCurve {keys [<at 0.2782 0.705 RCIM_Cubic>,<at 1 1 RCIM_Cubic>]}
        GreenCurve {keys [<at 0.2782 0.45 RCIM_Cubic>,<at 1 1 RCIM_Cubic>]}
        RedCurve {keys [<at 0.2782 3 RCIM_Cubic>,<at 1 1 RCIM_Cubic>]}
       }}))
    ScaleA=true
    ScaleRGB=true
   )
   &ScaleColor(
    Scale Alpha=&MergeHelper_FloatSwitch(
     0=&FloatFromCurve(FloatCurve=DataInterface() {propertyValues {Curve {keys [<at 0 0>,<at 0.05 1>]}}}, CurveIndex=$Emitter.Age)
     1=&FloatFromCurve(FloatCurve=DataInterface() {propertyValues {Curve {keys [<at 0.05 0>,<at 0.1 1>]}}}, CurveIndex=$Emitter.Age)
     2=&FloatFromCurve(FloatCurve=DataInterface() {propertyValues {Curve {keys [<at 0.1 0>,<at 0.15 1>]}}}, CurveIndex=$Emitter.Age)
     3=&FloatFromCurve(FloatCurve=DataInterface() {propertyValues {Curve {keys [<at 0.15 0>,<at 0.2 1>]}}}, CurveIndex=$Emitter.Age)
     4=&FloatFromCurve(FloatCurve=DataInterface() {propertyValues {Curve {keys [<at 0.2 0>,<at 0.25 1>]}}}, CurveIndex=$Emitter.Age)
     Index=&ModuloInt(A=$Particles.PartitionParticles.Partition, B=5)
     Num=::NewEnumerator3
    )
    ScaleA=true
    ScaleRGB=true
   ) {enabled=false, module=ScaleColor001}
  ]
  rendererClasses [&NiagaraSpriteRendererProperties]
  renderers [
   {rendererClass=&NiagaraSpriteRendererProperties}
   {rendererClass=&NiagaraSpriteRendererProperties, rendererIndex=1}
  ]
 }
 VentSmoke {
  emitterUpdateScript [
   &EmitterState(Life Cycle Mode=::NewEnumerator1, Loop Behavior=::NewEnumerator1, Loop Delay=0.05, Loop Duration=1, UseLoopDelay=true)
   &SpawnRate(
    SpawnRate=&FloatFromCurve(FloatCurve=DataInterface() {propertyValues {Curve {keys [<at 0.3 1>,<at 1 0.4>]}}}, CurveIndex=$Emitter.Age, Scale Curve=100)
    Use Spawn Probability=false
   )
   &SpawnBurst_Instantaneous(Spawn Count=10, Use Spawn Probability=false, UseLoopCountLimit=false) {enabled=false}
   /SetVariables(Emitter.ExhaustLifetime:NiagaraFloat=&FloatFromCurve(FloatCurve=DataInterface() {propertyValues {Curve {keys [<at 0 0>,<at 0.5 1>]}}}, CurveIndex=$Emitter.Age)) {isSetParameters=true, module=SetVariables_#2}
  ]
  particleSpawnScript [
   &InitializeParticle(
    Color=&FNTimeOfDayColor(Night Color=<rgba 0.7 0.7958 1 2>, Morning Color=<rgba 1 1 1 1>, Day Color=<rgba 1 1 1 1>, Evening Color=<rgba 1 1 1 1>)
    Lifetime=&RandomRangeFloat(Minimum=0.4, Maximum=0.5)
    Sprite Rotation Mode=::NewEnumerator2
    Sprite Size=&RandomRangeVector2D(Minimum=<5 5>, Maximum=<7 6>)
    Sprite Size Mode=::NewEnumerator4
    UsePositionOffset=false
    Write Material Random=true
   )
   &PartitionParticles(Number Of Partitions=&MakeIntFromFloat(Float=&FloatFromCurve(
      FloatCurve=DataInterface() {propertyValues {Curve {keys [<at 0 2 RCIM_Constant>,<at 0.05 4 RCIM_Constant>,<at 0.1 6 RCIM_Constant>,<at 0.15 8 RCIM_Constant>,<at 0.2 10 RCIM_Constant>]}}}
      CurveIndex=$Emitter.Age
     ))) {enabled=false, module=PartitionParticles001}
   &PartitionParticles(Number Of Partitions=10)
   &AddVelocity(Velocity=&MakeVector(
     X=-1
     Y=&MergeHelper_FloatSwitch(0=-0.7, 1=0.7, 2=?, 3=?, Index=&ModuloInt(A=$Particles.PartitionParticles.Partition, B=2), Num=::NewEnumerator0)
     Z=-0.1
    ))
   /SetVariables(Particles.PartitionParticles.Partition:NiagaraInt32=&MakeIntFromFloat(Float=&ClampFloat(
      Float=&MakeFloatFromInt(Int=&Add_Integer(
        A=$Particles.PartitionParticles.Partition
        B=&MakeIntFromFloat(Float=&FloatFromCurve(
          FloatCurve=DataInterface() {propertyValues {Curve {keys [<at 0.5 0 RCIM_Constant>,<at 0.6 2 RCIM_Constant>,<at 0.7 4 RCIM_Constant>,<at 0.8 6 RCIM_Constant>,<at 0.9 8 RCIM_Constant>,<at 1 10 RCIM_Constant>]}}}
          CurveIndex=$Emitter.Age
         ))
       ))
      Max=10
     ))) {isSetParameters=true, module=SetVariables_#3}
  ]
  particleUpdateScript [
   &SkeletalMeshLocation(
    Bone / Socket Index=$Particles.PartitionParticles.Partition
    Bone Sampling Mode=::NewEnumerator4
    MaskSpawnGroup=false
    Sampled Position Offset=<0 0 -2>
    Sampled Position Offset Coordinate Space:ENiagaraSkelMeshTransforms=::NewEnumerator2
    Skeletal Mesh=$System.Skeletal Mesh
   )
   /SetVariables(Particles.SpriteSizePuff:Vector2f=&Multiply_Vector2D_ByFloat(
     Float=&Multiply_Float(
      A=25
      B=&FloatFromCurve(FloatCurve=DataInterface() {propertyValues {Curve {keys [{interpMode=RCIM_Cubic, tangentMode=RCTM_User, time=0, value=0.1}, {interpMode=RCIM_Cubic, tangentMode=RCTM_User, time=1, value=1}]}}})
     )
     Vector2D=&Vector2DFromFloat(Value=&Lerp_Float(
       A=1
       B=0.5
       Alpha=&FloatFromCurve(FloatCurve=DataInterface() {propertyValues {Curve {keys [<at 0 0>,<at 1 1>]}}}, CurveIndex=$Emitter.Age)
      ))
    )) {isSetParameters=true}
   /SetVariables(Particles.PositionStream:NiagaraPosition=&AddVectorToPosition(
     Position=$Particles.Position
     Vector=&Multiply_VectorByFloat(
      Vector=&NormalizeVector(Vector To Normalize=$Particles.Initial.Velocity, Safe Normalize=true, Safe Fallback Threshold=0, Safe Fallback Vector=<0 0 0.001>)
      Float=&Lerp_Float(
       A=0
       B=10
       Alpha=&FloatFromCurve(
        FloatCurve=DataInterface() {propertyValues {Curve {keys [<at 0 0>,<at 1 1>]}}}
        Scale Curve=&RandomRangeFloat(Minimum=0.9, Maximum=1.1)
       )
      )
     )
    )) {isSetParameters=true}
   /SetVariables(Particles.PositionPuff:NiagaraPosition=&AddVectorToPosition(
     Position=$Particles.Position
     Vector=&Multiply_VectorByFloat(
      Vector=&NormalizeVector(Vector To Normalize=$Particles.Initial.Velocity, Safe Normalize=true, Safe Fallback Threshold=0, Safe Fallback Vector=<0 0 0.001>)
      Float=&Multiply_Float(
       A=&Multiply_Float(
        A=&Lerp_Float(
         A=13
         B=30
         Alpha=&FloatFromCurve(
          FloatCurve=DataInterface() {propertyValues {Curve {keys [<at 0 0>,<at 1 1>]}}}
          Scale Curve=&RandomRangeFloat(Minimum=0.9, Maximum=1.1)
         )
        )
        B=&Lerp_Float(A=1.5, B=1, Alpha=$Emitter.ExhaustLifetime)
       )
       B=&Lerp_Float(
        A=1.1
        B=0.5
        Alpha=&Divide_Float(
         A=&MakeFloatFromInt(Int=/Niagara/DynamicInputs/Divide/Divide_Int(A:NiagaraInt32=$Particles.PartitionParticles.Partition, B:NiagaraInt32=2))
         B=5
        )
       )
      )
     )
    )) {isSetParameters=true}
   &ParticleState()
   &ScaleSpriteSize(
    Scale Factor=&MakeVector2D(
     X=1
     Y=&Lerp_Float(
      A=1.1
      B=0.5
      Alpha=&Divide_Float(
       A=&MakeFloatFromInt(Int=&Divide_Int(A:NiagaraInt32=$Particles.PartitionParticles.Partition, B:NiagaraInt32=2))
       B=5
      )
     )
    )
    Scale Sprite Size Mode=::NewEnumerator2
   ) {module=ScaleSpriteSize001}
   &ScaleSpriteSize(
    Scale Factor=&MakeVector2D(
     X=&FloatFromCurve(FloatCurve=DataInterface() {propertyValues {Curve {keys [
          {interpMode=RCIM_Cubic, tangentMode=RCTM_User, time=0, value=0.5}
          {arriveTangent=0.7401, interpMode=RCIM_Cubic, leaveTangent=0.7401, tangentMode=RCTM_User, time=1, value=1}
         ]}}})
     Y=&FloatFromCurve(
      FloatCurve=DataInterface() {propertyValues {Curve {keys [{interpMode=RCIM_Cubic, tangentMode=RCTM_User, time=0, value=0.2}, {interpMode=RCIM_Cubic, tangentMode=RCTM_User, time=1, value=1}]}}}
      Scale Curve=&Lerp_Float(A=17, B=5, Alpha=$Emitter.ExhaustLifetime)
     )
    )
    Scale Sprite Size Mode=::NewEnumerator2
   )
   &Drag(Use Linear Drag=true, Use Rotational Drag=true) {enabled=false}
   &SolveForcesAndVelocity(Clamp Velocity=false, Limit Acceleration=false)
   &SubUVAnimation(
    "Sprite Renderer "=DataInterface() {propertyValues {SpriteRenderer=&self:VentSmoke_0.NiagaraSpriteRendererProperties_0}}
    UseEndFrame=false
    UseStartFrame=false
   )
   &ScaleColor(
    Scale Alpha=&FloatFromCurve(
     FloatCurve=DataInterface() {propertyValues {Curve {keys [
         {interpMode=RCIM_Cubic, tangentMode=RCTM_User, time=0, value=0}
         {interpMode=RCIM_Cubic, tangentMode=RCTM_User, time=0.15, value=1}
         {interpMode=RCIM_Cubic, tangentMode=RCTM_User, time=1, value=0}
        ]}}}
     Scale Curve=&RandomRangeFloat(Minimum=0.7, Maximum=1)
    )
    Scale RGB=&Lerp_Vector(
     A=&MakeVectorFromLinearColorRGB(Color=&ColorFromCurve(ColorCurve=DataInterface() {propertyValues {
         AlphaCurve {keys [<at 0 1 RCIM_Cubic>]}
         BlueCurve {keys [<at 0 0.76 RCIM_Cubic>,<at 0.8503 0.5 RCIM_Cubic>]}
         GreenCurve {keys [<at 0 0.4 RCIM_Cubic>,<at 0.8503 0.5 RCIM_Cubic>]}
         RedCurve {keys [<at 0 4 RCIM_Cubic>,<at 0.8503 0.6 RCIM_Cubic>]}
        }}))
     B=<0.6 0.5 0.5>
     Alpha=&FloatFromCurve(
      FloatCurve=DataInterface() {propertyValues {Curve {keys [
          {interpMode=RCIM_Cubic, tangentMode=RCTM_User, time=0.1, value=0}
          {arriveTangent=2.3957, interpMode=RCIM_Cubic, leaveTangent=2.3957, tangentMode=RCTM_User, time=0.8, value=1}
         ]}}}
      CurveIndex=$Emitter.Age
     )
    )
    ScaleA=true
    ScaleRGB=true
   )
   &ScaleColor(
    Scale Alpha=&MakeCustomFloatFromBool(
     Bool=&SetBoolByFloatComparison(
      A=&MakeFloatFromInt(Int=$Particles.PartitionParticles.Partition)
      B=&FloatFromCurve(
       FloatCurve=DataInterface() {propertyValues {Curve {keys [<at 0 2 RCIM_Constant>,<at 0.08 4 RCIM_Constant>,<at 0.16 6 RCIM_Constant>,<at 0.24 8 RCIM_Constant>,<at 0.32 10 RCIM_Constant>]}}}
       CurveIndex=$Emitter.Age
      )
      Comparison Type=::NewEnumerator7
     )
     True Float=1
     False Float=0
    )
    ScaleA=true
    ScaleRGB=false
   ) {module=ScaleColor001}
   &ScaleColor(
    Scale Alpha=&ClampFloat(Float=&Subtract_Float(
      A=&Add_Float(
       A=&MakeFloatFromInt(Int=&Divide_Int(A:NiagaraInt32=$Particles.PartitionParticles.Partition, B:NiagaraInt32=2))
       B=1
      )
      B=&Lerp_Float(
       A=0
       B=5
       Alpha=&FloatFromCurve(FloatCurve=DataInterface() {propertyValues {Curve {keys [<at 0.5 0>,<at 1 1>]}}}, CurveIndex=$Emitter.Age)
      )
     ))
    ScaleA=true
    ScaleRGB=false
   ) {enabled=false, module=ScaleColor002}
   &ScaleColor(
    Scale Alpha=&FloatFromCurve(
     FloatCurve=DataInterface() {propertyValues {Curve {keys [<at 0 1>,<at 1 0>]}}}
     CurveIndex=&Divide_Float(
      A=&MakeFloatFromInt(Int=&Divide_Int(A:NiagaraInt32=$Particles.PartitionParticles.Partition, B:NiagaraInt32=2))
      B=5
     )
    )
    ScaleA=true
    ScaleRGB=true
   ) {enabled=false, module=ScaleColor003}
   /SetVariables(Particles.ColorPuff:LinearColor=&MakeLinearColorFromVectorAndFloat(
     "Vector (RGB)"=&MakeVectorFromLinearColorRGB(Color=&LerpLinearColors(
       StartColor=$Particles.Color
       EndColor=<rgba 0.6 0.5 0.5 1>
       LerpFactor=&FloatFromCurve(
        FloatCurve=DataInterface() {propertyValues {Curve {keys [{interpMode=RCIM_Cubic, tangentMode=RCTM_User, time=0, value=0}, {interpMode=RCIM_Cubic, tangentMode=RCTM_User, time=1, value=1}]}}}
        CurveIndex=$Emitter.Age
       )
      ))
     "Float (Alpha)"=&Multiply_Float(
      A=&MakeFloatFromLinearColor(LinearColor=$Particles.Color, Channel:ENiagaraFloatFromLinearColorOptions=::NewEnumerator3)
      B=&FloatFromCurve(FloatCurve=DataInterface() {propertyValues {Curve {keys [{interpMode=RCIM_Cubic, tangentMode=RCTM_User, time=0.1, value=0}, {interpMode=RCIM_Cubic, tangentMode=RCTM_User, time=0.3, value=1}]}}})
     )
    )) {isSetParameters=true}
  ]
  rendererClasses [&NiagaraSpriteRendererProperties]
  renderers [
   {rendererClass=&NiagaraSpriteRendererProperties}
   {rendererClass=&NiagaraSpriteRendererProperties, rendererIndex=1}
  ]
 }
 VentSmoke001 {
  emitterUpdateScript [
   &EmitterState(Life Cycle Mode=::NewEnumerator1, Loop Behavior=::NewEnumerator1, Loop Delay=0.05, Loop Duration=2, UseLoopDelay=false)
   &SpawnRate(
    SpawnRate=&FloatFromCurve(FloatCurve=DataInterface() {propertyValues {Curve {keys [<at 0.3 1>,<at 1 0.4>]}}}, CurveIndex=$Emitter.Age, Scale Curve=100)
    Use Spawn Probability=false
   ) {enabled=false}
   &SpawnBurst_Instantaneous(Spawn Count=10, Use Spawn Probability=false, UseLoopCountLimit=false)
   /SetVariables(Emitter.ExhaustLifetime:NiagaraFloat=&FloatFromCurve(FloatCurve=DataInterface() {propertyValues {Curve {keys [<at 0 0>,<at 0.5 1>]}}}, CurveIndex=$Emitter.Age)) {enabled=false, isSetParameters=true, module=SetVariables_#2}
  ]
  particleSpawnScript [
   &InitializeParticle(
    Lifetime=3.3
    Sprite Rotation Mode=::NewEnumerator2
    Sprite Size Mode=::NewEnumerator3
    Uniform Sprite Size=30
    UsePositionOffset=false
    Write Material Random=true
   )
   &PartitionParticles(Number Of Partitions=&MakeIntFromFloat(Float=&FloatFromCurve(
      FloatCurve=DataInterface() {propertyValues {Curve {keys [<at 0 2 RCIM_Constant>,<at 0.05 4 RCIM_Constant>,<at 0.1 6 RCIM_Constant>,<at 0.15 8 RCIM_Constant>,<at 0.2 10 RCIM_Constant>]}}}
      CurveIndex=$Emitter.Age
     ))) {enabled=false, module=PartitionParticles001}
   &PartitionParticles(Number Of Partitions=10)
   /SetVariables(Particles.PartitionParticles.Partition:NiagaraInt32=&MakeIntFromFloat(Float=&ClampFloat(
      Float=&MakeFloatFromInt(Int=&Add_Integer(
        A=$Particles.PartitionParticles.Partition
        B=&MakeIntFromFloat(Float=&FloatFromCurve(
          FloatCurve=DataInterface() {propertyValues {Curve {keys [<at 0.5 0 RCIM_Constant>,<at 0.6 2 RCIM_Constant>,<at 0.7 4 RCIM_Constant>,<at 0.8 6 RCIM_Constant>,<at 0.9 8 RCIM_Constant>,<at 1 10 RCIM_Constant>]}}}
          CurveIndex=$Emitter.Age
         ))
       ))
      Max=10
     ))) {enabled=false, isSetParameters=true, module=SetVariables_#3}
   &CameraOffset(Camera Offset Amount=-3)
  ]
  particleUpdateScript [
   &SkeletalMeshLocation(
    Bone / Socket Index=$Particles.PartitionParticles.Partition
    Bone Sampling Mode=::NewEnumerator4
    MaskSpawnGroup=false
    Sampled Position Offset=&MakeVector(
     X=-4
     Y=&Multiply_Float(
      A=&MergeHelper_FloatSwitch(0=-7, 1=7, 2=?, 3=?, Index=&ModuloInt(A=$Particles.PartitionParticles.Partition, B=2), Num=::NewEnumerator0)
      B=&FloatFromCurve(
       FloatCurve=DataInterface() {propertyValues {Curve {keys [{interpMode=RCIM_Cubic, tangentMode=RCTM_User, time=0, value=1}, {interpMode=RCIM_Cubic, tangentMode=RCTM_User, time=0.55, value=0.5}]}}}
       Scale Curve=1.5
      )
     )
     Z=-2
    )
    Sampled Position Offset Coordinate Space:ENiagaraSkelMeshTransforms=::NewEnumerator2
    Skeletal Mesh=$System.Skeletal Mesh
   )
   &ScaleColor(
    Scale Alpha=&FloatFromCurve(
     FloatCurve=DataInterface() {propertyValues {Curve {keys [
         {interpMode=RCIM_Cubic, tangentMode=RCTM_User, time=0, value=0}
         {interpMode=RCIM_Cubic, tangentMode=RCTM_User, time=0.1, value=1}
         {interpMode=RCIM_Cubic, tangentMode=RCTM_User, time=1, value=0}
        ]}}}
     Scale Curve=2
    )
    Scale RGB=&MakeVectorFromLinearColorRGB(Color=&ColorFromCurve(ColorCurve=DataInterface() {propertyValues {
        AlphaCurve {keys [<at 0 1 RCIM_Cubic>]}
        BlueCurve {keys [<at 0.0779 3.16 RCIM_Cubic>,<at 0.3496 0.6>,<at 1 0.0333 RCIM_Cubic>]}
        GreenCurve {keys [<at 0.0779 2.4 RCIM_Cubic>,<at 0.3496 0.84>,<at 1 0 RCIM_Cubic>]}
        RedCurve {keys [<at 0.0779 10 RCIM_Cubic>,<at 0.3496 3>,<at 1 1 RCIM_Cubic>]}
       }}))
    ScaleA=true
    ScaleRGB=true
   )
   &ParticleState()
  ]
  rendererClasses [&NiagaraSpriteRendererProperties]
  renderers [{rendererClass=&NiagaraSpriteRendererProperties}]
 }
}
summary {
 systemName=NS_CherrySmoke_ExhaustVents
 emitters [
  {emitterName=VentSmoke, rendererClasses [&NiagaraSpriteRendererProperties]}
  {bEnabled=false, emitterName=Minimal001, rendererClasses [&NiagaraSpriteRendererProperties]}
  {emitterName=VentSmoke001, rendererClasses [&NiagaraSpriteRendererProperties]}
 ]
 userVariables []
}
