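# Niagara Data Channel handler that renders bullet impact effects for the SpireGun.
# Emitters spawn from impact data via shared FortImpacts modules, which supply position, normal, surface, ammo type and LOD distance.
# They include airflow meshes, cloud sprites, fading decals with a brief glow, and pink-tinted flashes with lights; Debug and standalone Light are disabled.
# Physical surface varies color, lifetime and motion, while ammo type and distance drive scale.
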NiagaraSystem /CherrySmokeWeaponGameplay/Gameplay/SpireGun/FX/NS_NDC_Handler_Impacts_Spire
dependencies {
 usedDataInterfaces [
  &NiagaraDataInterfaceArrayFloat
  &NiagaraDataInterfaceCurve
  &NiagaraDataInterfaceColorCurve
  &NiagaraDataInterfaceVectorCurve
  &NiagaraDataInterfaceVector2DCurve
  &NiagaraDataInterfaceArrayColor
  &NiagaraDataInterfaceArrayInt32
  /Script/Niagara.NiagaraDataInterfaceDataChannelRead
 ]
 usedDynamicInputs [&UniformRangedVector]
 usedModules [
  &SystemState
  &self:SystemSpawnScript.SystemScriptSource.SystemScriptGraph.NiagaraNodeAssignment_1.SetVariables_#1
  /Niagara/Modules/System/CompleteIfUnused
 ]
}
emitters {
 Airflow {
  emitterUpdateScript [
   &EmitterState()
   /Game/Effects/Niagara/DataChannels/Impacts/Spawn_NDCFortImpacts(Simple Spawn Override=1)
   &Spawn_NDCFortImpacts(Physical Material=SurfaceType3, Spawn Count=0, Surface Type=1) {module=Spawn_NDCFortImpacts001}
  ]
  particleSpawnScript [
   /Game/Effects/Niagara/DataChannels/Impacts/InitializeParticleFromNDC_FortImpact(
    AmmoType Scale=1
    Distance Scale=1
    FallbackPosition:NiagaraPosition=<0 0 1>
    Scale Array=DataInterface() {propertyValues {GpuSyncMode=SyncCpuToGpu, MaxElements=0, FloatData [0.75,1,1.2,0.3,1]}}
   )
   /SetVariables(Particles.IsSnowOrIce:NiagaraBool=&Boolean_OrOperation(
     Bool1=&SetBoolByIntComparison(A=$Particles.PhysicalSurfaceInteger, B=23, Comparison Type=::NewEnumerator4)
     Bool2=&SetBoolByIntComparison(A=$Particles.PhysicalSurfaceInteger, B=24, Comparison Type=::NewEnumerator4)
    )) {isSetParameters=true}
   &InitializeParticle(
    Color=&MakeCustomLinearColorFromBool(Bool=$Particles.IsSnowOrIce, True Color=<rgba 1.5 1.5 1.5 1>, False Color=<rgba 0.5922 0.5906 0.6 1>)
    Lifetime=/Game/Effects/Niagara/DataChannels/SelectFloatBySurface(
     Default:NiagaraFloat=&RandomRangeFloat(Minimum=0.25, Maximum=0.33)
     Surface:EPhysicalSurface=$Particles.PhysicalSurface
     Num Surfaces:NiagaraInt32=3
     Surface 0:EPhysicalSurface=SurfaceType23
     Value For Surface 0:NiagaraFloat=&RandomRangeFloat(Minimum=0.45, Maximum=0.5)
     Surface 1:EPhysicalSurface=SurfaceType24
     Value For Surface 1:NiagaraFloat=&RandomRangeFloat(Minimum=0.45, Maximum=0.5)
     Surface 2:EPhysicalSurface=SurfaceType26
     Value For Surface 2:NiagaraFloat=&RandomRangeFloat(Minimum=0.45, Maximum=0.5)
     Surface 3:EPhysicalSurface=?
     Value For Surface 3:NiagaraFloat=?
    )
    Mesh Scale Mode=::NewEnumerator3
    Mesh Uniform Scale=&Multiply_Float(A=&RandomRangeFloat(Minimum=1.8, Maximum=2), B=$Particles.AmmoTypeScale)
    Position=$Particles.Position
    Position Mode=::NewEnumerator1
    Sprite Size Mode=::NewEnumerator1
    Uniform Sprite Size Max=2.75
    Uniform Sprite Size Min=2
    UsePositionOffset=false
    Write Material Random=true
   )
   &CameraOffset(Camera Offset Amount=40)
   /SetVariables(Particles.AngleRange=0.03) {isSetParameters=true}
   &AddVelocityInCone(
    Cone Angle=&RandomRangeFloat(Minimum=0, Maximum=3)
    Cone Axis=$Particles.ImpactSurfaceNormal
    Cone Axis Coordinate Space=::Simulation
    Module.Use Velocity Falloff On Cone Axis=false
    Velocity Strength=&RandomRangeFloat(Minimum=0.055, Maximum=10)
   )
   &InitialMeshOrientation(
    Mesh Axis To Orient=<0 0 1>
    Mesh Orientation Mode=::NewEnumerator2
    Orientation Vector=$Particles.ImpactSurfaceNormal
    Rotation=&MakeVector(
     X=&Multiply_Float(A=$Particles.AngleRange, B=&RandomRangeFloat(Minimum=-1, Maximum=1))
     Y=&Multiply_Float(A=$Particles.AngleRange, B=&RandomRangeFloat(Minimum=-1, Maximum=1))
     Z=&RandomRangeFloat(Minimum=0, Maximum=1)
    )
    Use Rotation Vector=true
   )
  ]
  particleUpdateScript [
   &ParticleState()
   &ScaleColor(
    Scale Alpha=&FloatFromCurve(
     FloatCurve=DataInterface() {propertyValues {Curve {keys [
         {arriveTangent=9.0725, interpMode=RCIM_Cubic, leaveTangent=9.0725, tangentMode=RCTM_User, time=0, value=0}
         {arriveTangent=-2.5249, interpMode=RCIM_Cubic, leaveTangent=-2.5249, tangentMode=RCTM_User, time=0.1408, value=1}
         {arriveTangent=-0.0405, interpMode=RCIM_Cubic, leaveTangent=-0.0405, tangentMode=RCTM_User, time=1, value=0}
        ]}}}
     Scale Curve=&RandomRangeFloat(Minimum=15, Maximum=30)
    )
    Scale RGB=&MakeVectorFromLinearColorRGB(Color=&ColorFromCurve(ColorCurve=DataInterface() {propertyValues {
        AlphaCurve {keys []}
        BlueCurve {keys [
          {arriveTangent=0.0185, interpMode=RCIM_Cubic, time=0, value=0.4134}
          <at 0.4191 0.6533>
          {interpMode=RCIM_Cubic, leaveTangent=0.0185, time=1, value=0.1263}
         ]}
        GreenCurve {keys [
          {arriveTangent=-0.7509, interpMode=RCIM_Cubic, time=0, value=0.0961}
          <at 0.4191 0.64>
          {interpMode=RCIM_Cubic, leaveTangent=-0.7509, time=1, value=0.1263}
         ]}
        RedCurve {keys [
          {arriveTangent=-1.9276, interpMode=RCIM_Cubic, time=0, value=2}
          <at 0.4191 0.8>
          {interpMode=RCIM_Cubic, leaveTangent=-1.9276, time=1, value=0.2}
         ]}
       }}))
    ScaleA=true
    ScaleRGB=true
   )
   &ScaleMeshSize(Scale Factor=&VectorFromCurve(
     VectorCurve=DataInterface() {propertyValues {XCurve {keys [<at 0 0.75>,<at 1 0.75>]}, YCurve {keys [<at 0 0.75>,<at 1 0.75>]}, ZCurve {keys [<at 0 0.2>,<at 1 1>]}}}
     Scale Curve=&VectorFromFloat(Value=$Particles.DistanceScale)
    ))
   &ScaleMeshSize(Scale Factor=&VectorFromCurve(
     VectorCurve=DataInterface() {propertyValues {XCurve {keys [<at 0 1>,<at 1 1>]}, YCurve {keys [<at 0 1>,<at 1 1>]}, ZCurve {keys [<at 0 0.2>,<at 1 1>]}}}
     Scale Curve=&VectorFromFloat(Value=&RemapRange(Input Value To Remap=$Particles.LODDistance, Input Min=100, Input Max=8000, Output Min=1, Output Max=3))
    )) {enabled=false, module=ScaleMeshSize001}
   &SolveForcesAndVelocity(Clamp Velocity=false, Limit Acceleration=false)
   &DynamicMaterialParameters(
    Index 0 Param 1=&FloatFromCurve(FloatCurve=DataInterface() {propertyValues {Curve {keys [{interpMode=RCIM_Cubic, tangentMode=RCTM_User, time=0, value=0}, {interpMode=RCIM_Cubic, tangentMode=RCTM_User, time=1, value=1}]}}})
    Index 0 Param 2=&FloatFromCurve(FloatCurve=DataInterface() {propertyValues {Curve {keys [{tangentMode=RCTM_User, time=0, value=0.5}, {tangentMode=RCTM_User, time=1, value=1.5}]}}})
    Index 0 Param 3=1
    Index 0 Param 4=1
    Module.Index0Param0Write=true
    Module.Index0Param1Write=true
    Module.Index0Param2Write=true
    Module.Index0Param3Write=true
   )
  ]
  rendererClasses [&NiagaraMeshRendererProperties]
  renderers [{rendererClass=&NiagaraMeshRendererProperties, rendererIndex=0}]
 }
 Cloud {
  emitterUpdateScript [
   &EmitterState()
   &Spawn_NDCFortImpacts(Max Spawn Count=8, Min Spawn Count=6, Simple Max Spawn Count=2, Simple Min Spawn Count=1, Simple Spawn Override=2, Spawn Count Mode=2)
   &Spawn_NDCFortImpacts(Physical Material=SurfaceType3, Surface Type=1) {enabled=false, module=Spawn_NDCFortImpacts001}
  ]
  particleSpawnScript [
   &InitializeParticleFromNDC_FortImpact(
    AmmoType Scale=1
    FallbackPosition:NiagaraPosition=<0 0 1>
    Scale Array=DataInterface() {propertyValues {GpuSyncMode=SyncCpuToGpu, MaxElements=4, FloatData [0.8,1,1.2,0.7,1]}}
   )
   /SetVariables(Particles.VisibilityTag=/Game/Effects/Niagara/DataChannels/SelectIntBySurface(
     Default:NiagaraInt32=0
     Surface:EPhysicalSurface=$Particles.PhysicalSurface
     Num Surfaces:NiagaraInt32=4
     Surface 0:EPhysicalSurface=SurfaceType23
     Value For Surface 0:NiagaraInt32=1
     Surface 1:EPhysicalSurface=SurfaceType24
     Value For Surface 1:NiagaraInt32=1
     Surface 2:EPhysicalSurface=SurfaceType26
     Value For Surface 2:NiagaraInt32=2
     Surface 3:EPhysicalSurface=SurfaceType3
     Value For Surface 3:NiagaraInt32=3
    )) {isSetParameters=true}
   /SetVariables(
    Particles.LifetimeMax:NiagaraFloat=&SelectFloatBySurface(
     Default:NiagaraFloat=0.5
     Surface:EPhysicalSurface=$Particles.PhysicalSurface
     Num Surfaces:NiagaraInt32=4
     Surface 0:EPhysicalSurface=SurfaceType23
     Value For Surface 0:NiagaraFloat=1
     Surface 1:EPhysicalSurface=SurfaceType24
     Value For Surface 1:NiagaraFloat=0.4
     Surface 2:EPhysicalSurface=SurfaceType26
     Value For Surface 2:NiagaraFloat=1.25
     Surface 3:EPhysicalSurface=SurfaceType3
     Value For Surface 3:NiagaraFloat=2
    )
    Particles.LifetimeMin:NiagaraFloat=&SelectFloatBySurface(
     Default:NiagaraFloat=0.2
     Surface:EPhysicalSurface=$Particles.PhysicalSurface
     Num Surfaces:NiagaraInt32=4
     Surface 0:EPhysicalSurface=SurfaceType23
     Value For Surface 0:NiagaraFloat=0.5
     Surface 1:EPhysicalSurface=SurfaceType24
     Value For Surface 1:NiagaraFloat=0.2
     Surface 2:EPhysicalSurface=SurfaceType26
     Value For Surface 2:NiagaraFloat=0.5
     Surface 3:EPhysicalSurface=SurfaceType3
     Value For Surface 3:NiagaraFloat=1.75
    )
   ) {isSetParameters=true}
   /SetVariables(Particles.SizeMax=70, Particles.SizeMin=25) {isSetParameters=true}
   /SetVariables(Particles.SizeCurve:NiagaraFloat=&Multiply_Float(
     A=&Lerp_Float(
      A=$Particles.SizeMin
      B=$Particles.SizeMax
      Alpha=&FloatFromCurve(
       FloatCurve=DataInterface() {propertyValues {Curve {keys [
           {interpMode=RCIM_Cubic, tangentMode=RCTM_Break, time=0, value=0}
           {arriveTangent=2.5778, interpMode=RCIM_Cubic, tangentMode=RCTM_Break, time=1, value=1}
          ]}}}
       CurveIndex=&ReturnNormalizedExecIndex()
      )
     )
     B=&SelectFloatBySurface(
      Default:NiagaraFloat=1
      Surface:EPhysicalSurface=$Particles.PhysicalSurface
      Num Surfaces:NiagaraInt32=4
      Surface 0:EPhysicalSurface=SurfaceType26
      Value For Surface 0:NiagaraFloat=1.3
      Surface 1:EPhysicalSurface=SurfaceType24
      Value For Surface 1:NiagaraFloat=0.8
      Surface 2:EPhysicalSurface=SurfaceType23
      Value For Surface 2:NiagaraFloat=1.25
      Surface 3:EPhysicalSurface=SurfaceType3
      Value For Surface 3:NiagaraFloat=1
     )
    )) {isSetParameters=true}
   &InitializeParticle(
    Color Mode=::NewEnumerator0
    Lifetime=&Multiply_Float(
     A=&Lerp_Float(A=$Particles.LifetimeMin, B=$Particles.LifetimeMax, Alpha=&ReturnNormalizedExecIndex())
     B=$Particles.AmmoTypeScale
    )
    Position=$Particles.Position
    Position Mode=::NewEnumerator1
    Sprite Rotation Mode=::NewEnumerator2
    Sprite Size Mode=::NewEnumerator3
    Uniform Sprite Size=&Multiply_Float(A=$Particles.SizeCurve, B=$Particles.AmmoTypeScale)
    UsePositionOffset=false
    Write Material Random=true
   )
   /SetVariables(Particles.Color=&SelectLinearColorFromArray(
     Array Sampling Mode=::NewEnumerator1
     Direct Array Index=$Particles.VisibilityTag
     Color Selection Array=DataInterface() {propertyValues {ColorData [<rgba 0.66 0.5816 0.528 0.15>, <rgba 0.9 0.966 1 0.75>, <rgba 0.8 0.4304 0.272 0.75>, <rgba 0.25 0.25 0.25 0.5>]}}
     Random Seed=?
    )) {isSetParameters=true}
   /SetVariables(Particles.InitSpriteRotRate=&UniformRangedFloat(Minimum=-150, Maximum=150)) {isSetParameters=true}
   &AddVelocityInCone(
    Cone Angle=20
    Cone Axis=$Particles.ImpactSurfaceNormal
    Cone Axis Coordinate Space=::Simulation
    Module.Use Velocity Falloff On Cone Axis=false
    Velocity Strength=&Multiply_Float(
     A=&Lerp_Float(
      A=0
      B=500
      Alpha=&NormalizeFloat(Value=$Particles.SizeCurve, Min=$Particles.SizeMin, Max=$Particles.SizeMax)
     )
     B=$Particles.AmmoTypeScale
    )
   )
   &AddVelocity(
    Rotation Coordinate Space=::World
    Velocity=<0 0 1>
    Velocity Speed Scale=&Multiply_Float(A=100, B=$Particles.AmmoTypeScale)
   ) {module=AddVelocity001}
   /SetVariables(Particles.Velocity=/Game/Effects/Niagara/DataChannels/SelectVectorBySurface(
     Default:Vector3f=$Particles.Velocity
     Surface:EPhysicalSurface=$Particles.PhysicalSurface
     Num Surfaces:NiagaraInt32=1
     Surface 0:EPhysicalSurface=SurfaceType3
     Value For Surface 0:Vector3f=&Multiply_VectorByFloat(Vector=$Particles.Velocity, Float=&ReturnNormalizedExecIndex(Normalized Index Scale=0.25))
     Surface 1:EPhysicalSurface=?
     Value For Surface 1:Vector3f=?
     Surface 2:EPhysicalSurface=?
     Value For Surface 2:Vector3f=?
     Surface 3:EPhysicalSurface=?
     Value For Surface 3:Vector3f=?
    )) {isSetParameters=true}
   /SetVariables(Particles.DynamicMaterialParameter_Snow:Vector4f=<1 0.6 1 1>) {isSetParameters=true}
   /SetVariables(Particles.DynamicMaterialParameter_Sand:Vector4f=&MakeVector4(X=1, Y=0.66, Z=1, W=1)) {isSetParameters=true}
   /SetVariables(Particles.DynamicMaterialParameter_Smoke:Vector4f=&MakeVector4(X=0, Y=0.25, Z=1, W=1)) {isSetParameters=true}
   /SetVariables(Particles.SubUVRandom:NiagaraInt32=&RandomRangeInt(
     Minimum=0
     Maximum=&SelectIntBySurface(
      Default:NiagaraInt32=$Particles.PhysicalSurfaceInteger
      Surface:EPhysicalSurface=SurfaceType_Default
      Num Surfaces:NiagaraInt32=1
      Surface 0:EPhysicalSurface=SurfaceType3
      Value For Surface 0:NiagaraInt32=32
      Surface 1:EPhysicalSurface=?
      Value For Surface 1:NiagaraInt32=?
      Surface 2:EPhysicalSurface=?
      Value For Surface 2:NiagaraInt32=?
      Surface 3:EPhysicalSurface=?
      Value For Surface 3:NiagaraInt32=?
     )
    )) {isSetParameters=true}
  ]
  particleUpdateScript [
   &ParticleState()
   &Drag(
    Drag=&SelectFloatBySurface(
     Default:NiagaraFloat=5
     Surface:EPhysicalSurface=$Particles.PhysicalSurface
     Num Surfaces:NiagaraInt32=4
     Surface 0:EPhysicalSurface=SurfaceType23
     Value For Surface 0:NiagaraFloat=5
     Surface 1:EPhysicalSurface=SurfaceType24
     Value For Surface 1:NiagaraFloat=3
     Surface 2:EPhysicalSurface=SurfaceType26
     Value For Surface 2:NiagaraFloat=5
     Surface 3:EPhysicalSurface=SurfaceType3
     Value For Surface 3:NiagaraFloat=3
    )
    Use Linear Drag=true
    Use Rotational Drag=false
   )
   &GravityForce(Gravity=&SelectVectorBySurface(
     Default:Vector3f=<0 0 0>
     Surface:EPhysicalSurface=$Particles.PhysicalSurface
     Num Surfaces:NiagaraInt32=4
     Surface 0:EPhysicalSurface=SurfaceType23
     Value For Surface 0:Vector3f=<0 0 -980>
     Surface 1:EPhysicalSurface=SurfaceType24
     Value For Surface 1:Vector3f=<0 0 -490>
     Surface 2:EPhysicalSurface=SurfaceType26
     Value For Surface 2:Vector3f=&Multiply_VectorByFloat(Vector=<0 0 -980>, Float=1)
     Surface 3:EPhysicalSurface=SurfaceType3
     Value For Surface 3:Vector3f=&Multiply_VectorByFloat(Vector=<0 0 150>, Float=&ReturnNormalizedExecIndex())
    ))
   &SpriteRotationRate(Rotation Rate=&Multiply_Float(
     A=$Particles.InitSpriteRotRate
     B=&FloatFromCurve(
      FloatCurve=DataInterface() {propertyValues {Curve {keys [{arriveTangent=-2.0511, interpMode=RCIM_Cubic, leaveTangent=-2.0511, tangentMode=RCTM_User, time=0, value=1}, <at 1 0.1 RCIM_Cubic>]}}}
      Scale Curve=&SelectFloatBySurface(
       Default:NiagaraFloat=1
       Surface:EPhysicalSurface=$Particles.PhysicalSurface
       Num Surfaces:NiagaraInt32=3
       Surface 0:EPhysicalSurface=SurfaceType_Default
       Value For Surface 0:NiagaraFloat=1
       Surface 1:EPhysicalSurface=SurfaceType1
       Value For Surface 1:NiagaraFloat=1
       Surface 2:EPhysicalSurface=SurfaceType26
       Value For Surface 2:NiagaraFloat=1
       Surface 3:EPhysicalSurface=?
       Value For Surface 3:NiagaraFloat=?
      )
     )
    ))
   &SolveForcesAndVelocity(Clamp Velocity=false, Limit Acceleration=false)
   &SubUVAnimation(
    End Frame:NiagaraInt32=&Add_Integer(A=$Particles.SubUVRandom, B=63)
    Start Frame:NiagaraInt32=$Particles.SubUVRandom
    SubUV Loop Count:NiagaraFloat=&SelectFloatBySurface(
     Default:NiagaraFloat=0
     Surface:EPhysicalSurface=$Particles.PhysicalSurface
     Num Surfaces:NiagaraInt32=4
     Surface 0:EPhysicalSurface=SurfaceType23
     Value For Surface 0:NiagaraFloat=0.66
     Surface 1:EPhysicalSurface=SurfaceType24
     Value For Surface 1:NiagaraFloat=0.33
     Surface 2:EPhysicalSurface=SurfaceType26
     Value For Surface 2:NiagaraFloat=0.75
     Surface 3:EPhysicalSurface=SurfaceType3
     Value For Surface 3:NiagaraFloat=0.5
    )
   )
   &ScaleSpriteSize(Uniform Curve Sprite Scale=DataInterface() {propertyValues {Curve {keys [
        {arriveTangent=2.9194, interpMode=RCIM_Cubic, leaveTangent=2.9194, tangentMode=RCTM_User, time=0, value=0.25}
        {arriveTangent=1.1179, interpMode=RCIM_Cubic, leaveTangent=1.1179, tangentMode=RCTM_User, time=0.2, value=0.65}
        {arriveTangent=0.0224, interpMode=RCIM_Cubic, leaveTangent=0.0224, tangentMode=RCTM_User, time=1, value=1}
       ]}}})
   /Game/Effects/Niagara/Modules/Materials/MaterialSmoke(
    Erode:NiagaraFloat=&FloatFromCurve(FloatCurve=DataInterface() {propertyValues {
       bOptimizeLUT=false
       bUseLUT=false
       Curve {keys [
         {arriveTangent=-0.0652, interpMode=RCIM_Cubic, leaveTangent=-0.0652, tangentMode=RCTM_User, time=0.11, value=0.97}
         {arriveTangent=-5.7797, interpMode=RCIM_Cubic, leaveTangent=-5.7797, tangentMode=RCTM_User, time=0.39, value=0.66}
         {arriveTangent=-0.342, interpMode=RCIM_Cubic, leaveTangent=-0.342, tangentMode=RCTM_User, time=0.9954, value=-0.74}
        ]}
      }})
    Shading:NiagaraFloat=&FloatFromCurve(FloatCurve=DataInterface() {propertyValues {
       bOptimizeLUT=false
       bUseLUT=false
       Curve {keys [
         <at 0.07 0.5 RCIM_Cubic>
         {arriveTangent=1.2205, interpMode=RCIM_Cubic, leaveTangent=1.2205, time=0.17, value=1.02}
         {interpMode=RCIM_Cubic, tangentMode=RCTM_User, time=0.3895, value=0.89}
         <at 0.8 0 RCIM_Cubic>
        ]}
      }})
   ) {enabled=false}
   &DynamicMaterialParameters(
    Index 0 Param 1=&Add_Float(
     A=&RandomRangeFloat(Minimum=0, Maximum=1)
     B=&FloatFromCurve(
      FloatCurve=DataInterface() {propertyValues {Curve {keys [
          {arriveTangent=2.7533, interpMode=RCIM_Cubic, leaveTangent=1.8958, tangentMode=RCTM_User, time=0, value=1}
          {arriveTangent=0.8292, interpMode=RCIM_Cubic, leaveTangent=0.8292, tangentMode=RCTM_User, time=0.321, value=1.3453}
          {arriveTangent=0.0529, interpMode=RCIM_Cubic, leaveTangent=0.2021, tangentMode=RCTM_User, time=1, value=1.5}
         ]}}}
      CurveIndex=&Add_Float(A=&Multiply_Float(A=$Particles.Initial.MaterialRandom, B=0.5), B=$Particles.NormalizedAge)
     )
    )
    Index 0 Param 2=&FloatFromCurve(FloatCurve=DataInterface() {propertyValues {Curve {keys [
         {arriveTangent=0.8564, interpMode=RCIM_Cubic, leaveTangent=0.8564, tangentMode=RCTM_User, time=0.0046, value=-0.9753}
         {arriveTangent=4.2876, interpMode=RCIM_Cubic, leaveTangent=5.5268, tangentMode=RCTM_User, time=1, value=1}
        ]}}})
    Index 2 Param 1=&FloatFromCurve(FloatCurve=DataInterface() {propertyValues {Curve {keys [
         {interpMode=RCIM_Cubic, leaveTangent=4.2485, tangentMode=RCTM_Break, time=0, value=0}
         {arriveTangent=1.0548, interpMode=RCIM_Cubic, leaveTangent=1.0548, tangentMode=RCTM_User, time=0.17, value=0.5715}
         <at 1 1>
        ]}}})
    Index 2 Param 2=&RandomRangeFloat(Minimum=0, Maximum=1)
    Index 2 Param 3=&FloatFromCurve(
     FloatCurve=DataInterface() {propertyValues {Curve {keys [
         {interpMode=RCIM_Cubic, leaveTangentWeight=0.3354, tangentMode=RCTM_Break, time=0.4, value=0}
         {arriveTangent=2.7846, interpMode=RCIM_Cubic, leaveTangent=11.1233, tangentMode=RCTM_Break, time=1, value=1}
        ]}}}
     Scale Curve=7
    )
    Index 2 Param 4=&FloatFromCurve(
     FloatCurve=DataInterface() {propertyValues {Curve {keys [
         {interpMode=RCIM_Cubic, leaveTangentWeight=0.4044, tangentMode=RCTM_Break, time=0, value=0.8}
         {arriveTangent=0.4016, arriveTangentWeight=0.382, interpMode=RCIM_Cubic, tangentMode=RCTM_Break, time=1, value=1}
        ]}}}
     Scale Curve=0.5
    )
    Module.Param0WriteEnabled=true
    Module.Param0WriteEnabled2=true
    Module.Param1WriteEnabled=true
    Module.Param1WriteEnabled2=true
    Module.Param2WriteEnabled=true
    Module.Param2WriteEnabled2=true
    Module.Param3WriteEnabled=true
    Module.Param3WriteEnabled2=true
    Write Parameter Index 2=true
   )
   /SetVariables(Particles.DynamicMaterialParameter_Sand:Vector4f=&MakeVector4(
     X=&Add_Float(
      A=&RandomRangeFloat(Minimum=0, Maximum=1)
      B=&FloatFromCurve(
       FloatCurve=DataInterface() {propertyValues {Curve {keys [
           {arriveTangent=0.7934, interpMode=RCIM_Cubic, leaveTangent=0.7934, tangentMode=RCTM_User, time=0, value=1}
           {interpMode=RCIM_Cubic, leaveTangent=0.2021, time=1, value=1.5}
          ]}}}
       CurveIndex=&Add_Float(A=&Multiply_Float(A=$Particles.Initial.MaterialRandom, B=0.5), B=$Particles.NormalizedAge)
      )
     )
     Y=0
     Z=0
     W=&FloatFromCurve(FloatCurve=DataInterface() {propertyValues {Curve {keys [<at 0 1 RCIM_Cubic>,<at 1 0.25>]}}})
    )) {enabled=false, isSetParameters=true}
   /SetVariables(Particles.DynamicMaterialParameter2_Sand:Vector4f=&MakeVector4(
     X=&FloatFromCurve(FloatCurve=DataInterface() {propertyValues {Curve {keys [{tangentMode=RCTM_User, time=0, value=0}, {tangentMode=RCTM_User, time=1, value=1}]}}})
     Y=&RandomRangeFloat(Minimum=-0.1, Maximum=0.1)
     Z=&Lerp_Float(
      A=1
      B=6
      Alpha=&FloatFromCurve(FloatCurve=DataInterface() {propertyValues {Curve {keys [{interpMode=RCIM_Cubic, tangentMode=RCTM_User, time=0.33, value=0}, {interpMode=RCIM_Cubic, tangentMode=RCTM_User, time=0.66, value=1}]}}})
     )
     W=&Multiply_Float(
      A=&RandomRangeFloat(Minimum=1.5, Maximum=1.75)
      B=&Lerp_Float(
       A=1.25
       B=1
       Alpha=&FloatFromCurve(FloatCurve=DataInterface() {propertyValues {Curve {keys [{tangentMode=RCTM_User, time=0, value=0}, {tangentMode=RCTM_User, time=1, value=1}]}}})
      )
     )
    )) {enabled=false, isSetParameters=true}
   &ScaleColor(
    Scale Alpha=&FloatFromCurve(FloatCurve=DataInterface() {propertyValues {Curve {keys [
         {interpMode=RCIM_Cubic, leaveTangent=17.2181, tangentMode=RCTM_Break, time=0, value=0}
         {interpMode=RCIM_Cubic, tangentMode=RCTM_Break, time=0.1, value=1}
         {arriveTangent=-1.1111, interpMode=RCIM_Cubic, leaveTangent=-1.1111, time=0.5, value=0.2}
         <at 1 0 RCIM_Cubic>
        ]}}})
    Scale RGB=&MakeVectorFromLinearColorRGB(Color=&ColorFromCurve(ColorCurve=DataInterface() {propertyValues {
        AlphaCurve {keys [<at 0 1 RCIM_Cubic>]}
        BlueCurve {keys [<at 0.0474 0.325 RCIM_Cubic>,<at 0.463 1 RCIM_Cubic>]}
        GreenCurve {keys [<at 0.0474 0.1 RCIM_Cubic>,<at 0.463 1 RCIM_Cubic>]}
        RedCurve {keys [<at 0.0474 1 RCIM_Cubic>,<at 0.463 1 RCIM_Cubic>]}
       }}))
    ScaleA=true
    ScaleRGB=true
   )
   &CameraOffset(Camera Offset Amount=&Multiply_Float(A=&MakeFloatFromVector2d(Vector 2D=$Particles.SpriteSize), B=0.5))
  ]
  rendererClasses [&NiagaraSpriteRendererProperties]
  renderers [
   {rendererClass=&NiagaraSpriteRendererProperties}
   {rendererClass=&NiagaraSpriteRendererProperties, rendererIndex=1}
   {rendererClass=&NiagaraSpriteRendererProperties, rendererIndex=2}
   {rendererClass=&NiagaraSpriteRendererProperties, rendererIndex=3}
   {rendererClass=&NiagaraSpriteRendererProperties, rendererIndex=4}
   {rendererClass=&NiagaraSpriteRendererProperties, rendererIndex=5}
   {rendererClass=&NiagaraSpriteRendererProperties, rendererIndex=6}
  ]
 }
 Debug {
  bEnabled=false
  emitterUpdateScript [
   &EmitterState()
   &Spawn_NDCFortImpacts()
  ]
  particleSpawnScript [
   &InitializeParticleFromNDC_FortImpact(FallbackPosition:NiagaraPosition=<0 0 1>)
   &InitializeParticle(
    Color=&MakeLinearColor(
     R:NiagaraFloat=&MakeFloatFromInt(Int=$Output.InitializeParticleFromNDC_FortImpact.ImpactPhysicalSurfaceInteger)
     G:NiagaraFloat=$Output.InitializeParticleFromNDC_FortImpact.ImpactDamage
     B:NiagaraFloat=&MakeCustomFloatFromBool(Bool=$Output.InitializeParticleFromNDC_FortImpact.bSimple, True Float=1, False Float=0)
     A:NiagaraFloat=$Output.InitializeParticleFromNDC_FortImpact.LODDistance
    )
    Position=$Output.InitializeParticleFromNDC_FortImpact.ImpactPoint
    Position Mode=::NewEnumerator1
    Sprite Size Mode=::NewEnumerator3
    Uniform Sprite Size=100
    UsePositionOffset=false
    Write Material Random=true
   )
  ]
  particleUpdateScript [
   &ParticleState()
   &CameraOffset(Camera Offset Amount=50)
  ]
  rendererClasses [&NiagaraSpriteRendererProperties]
  renderers [{rendererClass=&NiagaraSpriteRendererProperties}]
 }
 Decal {
  emitterUpdateScript [
   &EmitterState(Scalability Mode=::NewEnumerator1)
   &Spawn_NDCFortImpacts()
   &Spawn_NDCFortImpacts(Physical Material=SurfaceType8, Spawn Count=0, Surface Type=1) {module=Spawn_NDCFortImpacts001}
  ]
  particleSpawnScript [
   &InitializeParticleFromNDC_FortImpact(
    AmmoType Scale=1
    FallbackImpactScale=50
    Scale Array=DataInterface() {propertyValues {GpuSyncMode=SyncCpuToGpu, MaxElements=0, FloatData [12,17.5,25,10,40]}}
   )
   /SetVariables(Particles.VisibilityTag=&SelectIntFromArray(Direct Array Index=$Particles.PhysicalSurfaceInteger, Int Selection Array=DataInterface() {propertyValues {IntData [0,1,2,3,0]}})) {isSetParameters=true}
   /SetVariables(Particles.AmmoTypeScale:NiagaraFloat=&SelectFloatFromArray(Direct Array Index=$Particles.AmmoType, Float Selection Array=DataInterface() {propertyValues {FloatData [12,17.5,25,10,40]}})) {enabled=false, isSetParameters=true}
   &InitializeParticle(
    Lifetime=&SelectFloatFromArray(Direct Array Index=$Engine.QualityLevel, Float Selection Array=DataInterface() {propertyValues {FloatData [2,3,4,5]}})
    Mesh Scale Mode=::NewEnumerator3
    Mesh Uniform Scale=&Multiply_Float(A=$Particles.AmmoTypeScale, B=&RandomRangeFloat(Minimum=0.85, Maximum=1))
    Position=$Particles.Position
    Position Mode=::NewEnumerator1
    Position Offset=<20 0 0>
    UsePositionOffset=false
    Write Material Random=true
   )
   /SetVariables(Particles.DecalEmissiveColor:LinearColor=&SelectLinearColorFromArray(
     Array Sampling Mode=::NewEnumerator1
     Direct Array Index=1
     Color Selection Array=DataInterface() {propertyValues {ColorData [<rgba 0 0 0 1>, <rgba 4 0.5 0.1 1>, <rgba 10 0.6 0.1 1>, <rgba 0.5 3 4 1>, <rgba 1 5 15 1>]}}
     Random Seed=?
    )) {enabled=false, isSetParameters=true}
   /SetVariables(Particles.DecalEmissiveColor:LinearColor=&SelectLinearColorFromArray(
     Array Sampling Mode=::NewEnumerator1
     Direct Array Index=1
     Color Selection Array=DataInterface() {propertyValues {ColorData [<rgba 0 0 0 1>, <rgba 5 0.3 0.77 1>, <rgba 10 0.6 0.1 1>, <rgba 0.5 3 4 1>, <rgba 1 5 15 1>]}}
     Random Seed=?
    )) {isSetParameters=true}
   /Niagara/Modules/Spawn/Decal/InitialDecalOrientation(
    Orientation Vector:Vector3f=$Particles.ImpactSurfaceNormal
    Rotation:Vector3f=&RandomRangeVector(Minimum=<0 0 0>, Maximum=<0 0 1>)
    Use Rotation Vector=true
   )
  ]
  particleUpdateScript [
   &ParticleState()
   &ScaleColor(
    Scale Alpha=&FloatFromCurve(FloatCurve=DataInterface() {propertyValues {Curve {keys [
         {tangentMode=RCTM_User, time=0, value=1}
         {arriveTangent=-0.0405, interpMode=RCIM_Cubic, leaveTangent=-0.0405, tangentMode=RCTM_User, time=0.7009, value=1}
         {interpMode=RCIM_Cubic, leaveTangent=-8.0089, time=1, value=0}
        ]}}})
    Scale RGB=&Multiply_VectorByFloat(
     Vector=&MakeVectorFromLinearColorRGB(Color=$Particles.DecalEmissiveColor)
     Float=&FloatFromCurve(FloatCurve=DataInterface() {propertyValues {Curve {keys [
          {arriveTangent=-27.9932, interpMode=RCIM_Cubic, leaveTangent=-27.9932, tangentMode=RCTM_User, time=0, value=2}
          {arriveTangent=-0.2721, interpMode=RCIM_Cubic, leaveTangent=-0.2721, tangentMode=RCTM_User, time=0.05, value=0.025}
          <at 0.15 0 RCIM_Cubic>
         ]}}})
    )
    ScaleA=true
    ScaleRGB=true
   )
   &ScaleMeshSize(Scale Factor=<5 5 5>) {enabled=false}
  ]
  rendererClasses [&NiagaraDecalRendererProperties]
  renderers [
   {rendererClass=&NiagaraDecalRendererProperties, rendererIndex=0}
   {rendererClass=&NiagaraDecalRendererProperties, rendererIndex=1}
   {rendererClass=&NiagaraDecalRendererProperties, rendererIndex=2}
   {rendererClass=&NiagaraDecalRendererProperties, rendererIndex=3}
  ]
 }
 Flash {
  emitterUpdateScript [
   &EmitterState()
   &Spawn_NDCFortImpacts()
  ]
  particleSpawnScript [
   &InitializeParticleFromNDC_FortImpact(
    AmmoType Scale=1
    Distance Scale=1
    FallbackPosition:NiagaraPosition=<0 0 1>
    Scale Array=DataInterface() {propertyValues {GpuSyncMode=SyncCpuToGpu, MaxElements=4, FloatData [0.8,1,1.25,0.6,1]}}
   )
   /SetVariables() {isSetParameters=true}
   &InitializeParticle(
    Lifetime Max=0.2
    Lifetime Min=0.133
    Lifetime Mode=::NewEnumerator1
    Mesh Scale=&Multiply_VectorByFloat(
     Vector=&MakeVector(
      X=&RandomRangeFloat(Minimum=0.5, Maximum=1)
      Y=&RandomRangeFloat(Minimum=0.6, Maximum=1)
      Z=&RandomRangeFloat(Minimum=0.6, Maximum=1)
     )
     Float=&Multiply_Float(
      A=$Particles.AmmoTypeScale
      B=&SelectFloatBySurface(
       Default:NiagaraFloat=1
       Surface:EPhysicalSurface=$Particles.PhysicalSurface
       Num Surfaces:NiagaraInt32=1
       Surface 0:EPhysicalSurface=SurfaceType3
       Value For Surface 0:NiagaraFloat=0.66
       Surface 1:EPhysicalSurface=?
       Value For Surface 1:NiagaraFloat=?
       Surface 2:EPhysicalSurface=?
       Value For Surface 2:NiagaraFloat=?
       Surface 3:EPhysicalSurface=?
       Value For Surface 3:NiagaraFloat=?
      )
     )
    )
    Mesh Scale Mode=::NewEnumerator4
    Position=$Particles.Position
    Position Mode=::NewEnumerator1
    Sprite Rotation Mode=::NewEnumerator2
    Sprite Size Mode=::NewEnumerator3
    Uniform Sprite Size=&Multiply_Float(A=&RandomRangeFloat(Minimum=65, Maximum=80), B=$Particles.AmmoTypeScale)
    UsePositionOffset=false
    Write Material Random=true
   )
   /SetVariables(Particles.AngleRange=0.02) {isSetParameters=true}
   &SubUVAnimation(End Frame=3, SubUV Animation Mode:ENiagara_SubUVLookupMode=::NewEnumerator2)
   &InitialMeshOrientation(
    Mesh Orientation Mode=::NewEnumerator2
    Orientation Vector=$Particles.ImpactSurfaceNormal
    Rotation=&MakeVector(
     X=&RandomRangeFloat(Minimum=0, Maximum=1)
     Y=&Multiply_Float(A=$Particles.AngleRange, B=&RandomRangeFloat(Minimum=-1, Maximum=1))
     Z=&Multiply_Float(A=$Particles.AngleRange, B=&RandomRangeFloat(Minimum=-1, Maximum=1))
    )
    Use Rotation Vector=true
   )
   /SetVariables(Particles.LightEnabled:NiagaraBool=&SetBoolByFloatComparison(
     A=$Particles.LODDistance
     B=&Multiply_Float(A=8000, B=$Particles.AmmoTypeScale)
     Comparison Type=::NewEnumerator7
    )) {isSetParameters=true}
   /SetVariables(Particles.LightPosition:NiagaraPosition=&AddVectorToPosition(
     Position=$Particles.Position
     Vector=&Multiply_VectorByFloat(Vector=$Particles.ImpactSurfaceNormal, Float=&Multiply_Float(A=20, B=$Particles.AmmoTypeScale))
    )) {isSetParameters=true}
   /SetVariables(Particles.LightVolumetricScattering=1) {isSetParameters=true}
   /SetVariables(Particles.LightRadius=&Multiply_Float(A=&Multiply_Float(A=50, B=$Particles.DistanceScale), B=$Particles.AmmoTypeScale)) {isSetParameters=true}
   /SetVariables(Particles.LightExponent=6) {isSetParameters=true}
   &CameraOffset(Camera Offset Amount=&Multiply_Float(A=20, B=$Particles.DistanceScale))
  ]
  particleUpdateScript [
   &ParticleState()
   &Color(
    Color=&MakeLinearColorFromVectorAndFloat(
     "Vector (RGB)"=&MakeVectorFromLinearColorRGB(Color=&ColorFromCurve(ColorCurve=DataInterface() {propertyValues {
         AlphaCurve {keys []}
         BlueCurve {keys [
           {arriveTangent=-1.4036, interpMode=RCIM_Cubic, time=0, value=5.1}
           <at 0.2204 1.6176>
           {interpMode=RCIM_Cubic, leaveTangent=-1.4036, time=1, value=0.001}
          ]}
         GreenCurve {keys [
           {arriveTangent=-5.857, interpMode=RCIM_Cubic, time=0, value=0.2}
           <at 0.2204 0.1759>
           {interpMode=RCIM_Cubic, leaveTangent=-5.857, time=1, value=0.001}
          ]}
         RedCurve {keys [
           {arriveTangent=-14.4701, interpMode=RCIM_Cubic, time=0, value=10}
           <at 0.2204 8.8265>
           {interpMode=RCIM_Cubic, leaveTangent=-14.4701, time=1, value=0.1}
          ]}
        }}))
     "Float (Alpha)"=&FloatFromCurve(FloatCurve=DataInterface() {propertyValues {Curve {keys [{arriveTangent=-2.7272, interpMode=RCIM_Cubic, time=0, value=1}, {interpMode=RCIM_Cubic, tangentMode=RCTM_User, time=1, value=0}]}}})
    )
    Scale Color=&VectorFromFloat(Value=&FloatFromCurve(FloatCurve=DataInterface() {propertyValues {Curve {keys [{interpMode=RCIM_Cubic, tangentMode=RCTM_User, time=0, value=1}, {interpMode=RCIM_Cubic, tangentMode=RCTM_User, time=1, value=0}]}}}))
   )
   &ScaleMeshSize(Scale Factor=&Multiply_VectorByFloat(
     Vector=&VectorFromCurve(VectorCurve=DataInterface() {propertyValues {
        XCurve {keys [<at 0 1>,<at 1 1>]}
        YCurve {keys [{arriveTangent=1.6329, interpMode=RCIM_Cubic, time=0.1, value=0.2}, {interpMode=RCIM_Cubic, leaveTangent=1.0526, time=1, value=1}]}
        ZCurve {keys [{arriveTangent=1.6329, interpMode=RCIM_Cubic, time=0.1, value=0.2}, {interpMode=RCIM_Cubic, leaveTangent=1.0526, time=1, value=1}]}
       }})
     Float=&Multiply_Float(
      A=$Particles.DistanceScale
      B=&FloatFromCurve(FloatCurve=DataInterface() {propertyValues {Curve {keys [{arriveTangent=1.9165, interpMode=RCIM_Cubic, leaveTangent=1.9165, tangentMode=RCTM_User, time=0, value=0.25}, <at 1 1>]}}})
     )
    ))
   &FN_Sprite_Attributes(
    Create SpriteSize Attribute=true
    SpriteSize=&Multiply_Vector2D_ByFloat(
     Float=$Particles.DistanceScale
     Vector2D=&Multiply_Vector2D_ByFloat(
      Float=&FloatFromCurve(FloatCurve=DataInterface() {propertyValues {Curve {keys [
           {arriveTangent=1.126, interpMode=RCIM_Cubic, leaveTangent=1.126, tangentMode=RCTM_User, time=0, value=0.25}
           {arriveTangent=0.5184, tangentMode=RCTM_User, time=1, value=1}
          ]}}})
      Vector2D=$Particles.Initial.SpriteSize
     )
    )
    Write to Module Namespace=true
   ) {module=FN_Sprite}
   /SetVariables(Particles.SpriteColor=&MakeLinearColorFromVectorAndFloat(
     "Vector (RGB)"=&MakeVectorFromLinearColorRGB(Color=&ColorFromCurve(ColorCurve=DataInterface() {propertyValues {
         AlphaCurve {keys []}
         BlueCurve {keys [
           {arriveTangent=-1.4036, interpMode=RCIM_Cubic, time=0.0469, value=2.5583}
           <at 0.3341 0.61>
           {interpMode=RCIM_Cubic, leaveTangent=-1.4036, time=1, value=0.0059}
          ]}
         GreenCurve {keys [
           {arriveTangent=-5.857, interpMode=RCIM_Cubic, time=0.0469, value=0.5}
           <at 0.3341 0.12>
           {interpMode=RCIM_Cubic, leaveTangent=-5.857, time=1, value=0.0059}
          ]}
         RedCurve {keys [
           {arriveTangent=-14.4701, interpMode=RCIM_Cubic, time=0.0469, value=10}
           <at 0.3341 6>
           {interpMode=RCIM_Cubic, leaveTangent=-14.4701, time=1, value=0.5625}
          ]}
        }}))
     "Float (Alpha)"=&FloatFromCurve(FloatCurve=DataInterface() {propertyValues {Curve {keys [{interpMode=RCIM_Cubic, tangentMode=RCTM_Break, time=0, value=1}, {interpMode=RCIM_Cubic, tangentMode=RCTM_Break, time=1, value=0}]}}})
    )) {isSetParameters=true}
   /SetVariables(Particles.Light_Color:LinearColor=&MultiplyLinearColor(
     Linear Color=<rgba 1.5 0.015 0.1635 1>
     Scale Factor=&MakeFloat4FromVectorAndFloat(
      "Vector (RGB)"=<1 1 1>
      "Float (Alpha)"=&Multiply_Float(
       A=18
       B=&FloatFromCurve(
        FloatCurve=DataInterface() {propertyValues {Curve {keys [
            {interpMode=RCIM_Cubic, tangentMode=RCTM_User, time=0, value=0}
            {interpMode=RCIM_Cubic, tangentMode=RCTM_User, time=0.2, value=1}
            {tangentMode=RCTM_User, time=1, value=0}
           ]}}}
        Scale Curve=&FloatFromCurve(
         FloatCurve=DataInterface() {propertyValues {Curve {keys [{interpMode=RCIM_Cubic, tangentMode=RCTM_User, time=0, value=1}, {interpMode=RCIM_Cubic, tangentMode=RCTM_User, time=1, value=0}]}}}
         CurveIndex=&RemapRange(Input Value To Remap=$Particles.LODDistance, Input Min=0, Input Max=2000, Output Min=0, Output Max=1)
        )
       )
      )
     )
    )) {isSetParameters=true}
   /SetVariables(
    Particles.ColorOuter:LinearColor=&MultiplyLinearColor(Linear Color=$Particles.Color, Scale Factor=<0.15 0 0.15 0.5>)
    Particles.ScaleMeshOuter:Vector3f=&Multiply_Vector(A=$Particles.Scale, B=<1.35 1.2 1.2>)
   ) {isSetParameters=true}
   /SetVariables(
    Particles.SpriteColorOuter:LinearColor=&MultiplyLinearColor(Linear Color=$Particles.SpriteColor, Scale Factor=<0 0 0 1>)
    Particles.SpriteSizeOuter:Vector2f=&Multiply_Vector2D(A=$Particles.FN_Sprite.SpriteSize, B=<2.5 2.5>)
   ) {isSetParameters=true}
  ]
  rendererClasses [
   &NiagaraMeshRendererProperties
   &NiagaraSpriteRendererProperties
   &NiagaraLightRendererProperties
  ]
  renderers [
   {rendererClass=&NiagaraMeshRendererProperties, rendererIndex=0}
   {rendererClass=&NiagaraSpriteRendererProperties, rendererIndex=1}
   {rendererClass=&NiagaraLightRendererProperties, rendererIndex=2}
   {rendererClass=&NiagaraMeshRendererProperties, rendererIndex=3}
   {rendererClass=&NiagaraSpriteRendererProperties, rendererIndex=4}
  ]
 }
 Light {
  bEnabled=false
  emitterUpdateScript [
   &EmitterState()
   &Spawn_NDCFortImpacts(Max Spawn Count=1, Min Spawn Count=1, Reader:NiagaraDataInterfaceDataChannelRead=$System.NDCReader)
   /Game/Effects/Niagara/DataChannels/Impacts/SimpleFilter_NDCFortImpacts(Exclude=false) {enabled=false}
   /Game/Effects/Niagara/DataChannels/Impacts/LODDistanceFilter_NDCFortImpacts(LODDistance=5000)
  ]
  particleSpawnScript [
   &InitializeParticleFromNDC_FortImpact()
   /SetVariables(Particles.AmmoTypeScale:NiagaraFloat=&SelectFloatFromArray(
     Direct Array Index=$Output.InitializeParticleFromNDC_FortImpact.AmmoType
     Float Selection Array=DataInterface() {propertyValues {FloatData [0.75,1,1.25,0.6,1]}}
    )) {isSetParameters=true}
   &InitializeParticle(
    Lifetime=0.17
    Position=$Output.InitializeParticleFromNDC_FortImpact.ImpactPoint
    Position Mode=::NewEnumerator1
    Position Offset=&Multiply_VectorByFloat(
     Vector=$Output.InitializeParticleFromNDC_FortImpact.ImpactNormal
     Float=&Multiply_Float(A=20, B=$Particles.AmmoTypeScale)
    )
    Sprite Size Mode=::NewEnumerator3
    Uniform Sprite Size=120
    UsePositionOffset=true
    Write Material Random=true
   )
   /SetVariables(Particles.LightExponent=6) {isSetParameters=true}
   /SetVariables(Particles.LightRadius=&Multiply_Float(
     A=&Multiply_Float(
      A=50
      B=&RemapRange(Input Value To Remap=$Output.InitializeParticleFromNDC_FortImpact.LODDistance, Input Min=100, Input Max=8000, Output Min=1, Output Max=2)
     )
     B=$Particles.AmmoTypeScale
    )) {isSetParameters=true}
   /SetVariables(Particles.LightVolumetricScattering=1) {isSetParameters=true}
  ]
  particleUpdateScript [
   &ParticleState()
   /SetVariables(Particles.Light_Color:LinearColor=&MultiplyLinearColor(
     Linear Color=<rgba 1 0.325 0.1 1>
     Scale Factor=&MakeFloat4FromVectorAndFloat(
      "Vector (RGB)"=<1 1 1>
      "Float (Alpha)"=&Multiply_Float(
       A=15
       B=&FloatFromCurve(
        FloatCurve=DataInterface() {propertyValues {Curve {keys [
            {interpMode=RCIM_Cubic, tangentMode=RCTM_User, time=0, value=0}
            {interpMode=RCIM_Cubic, tangentMode=RCTM_User, time=0.2, value=1}
            {tangentMode=RCTM_User, time=1, value=0}
           ]}}}
        Scale Curve=&FloatFromCurve(
         FloatCurve=DataInterface() {propertyValues {Curve {keys [{interpMode=RCIM_Cubic, tangentMode=RCTM_User, time=0, value=1}, {interpMode=RCIM_Cubic, tangentMode=RCTM_User, time=1, value=0}]}}}
         CurveIndex=&RemapRange(Input Value To Remap=$Output.InitializeParticleFromNDC_FortImpact.LODDistance, Input Min=0, Input Max=2000, Output Min=0, Output Max=1)
        )
       )
      )
     )
    )) {isSetParameters=true}
  ]
  rendererClasses [&NiagaraLightRendererProperties]
  renderers [{rendererClass=&NiagaraLightRendererProperties, rendererIndex=0}]
 }
 NE_FN_PhysMatDebris {
  emitterUpdateScript [
   &EmitterState()
   &Spawn_NDCFortImpacts(Distance Culling=1, LOD Distance=5000, Spawn Count Mode=0)
   &Spawn_NDCFortImpacts(
    Max Spawn Count=5
    Min Spawn Count=3
    Simple Max Spawn Count=2
    Simple Min Spawn Count=1
    Simple Spawn Override=2
    Spawn Count Mode=2
    Surface Type=1
   ) {module=Spawn_NDCFortImpacts001}
   &Spawn_NDCFortImpacts(
    Max Spawn Count=5
    Min Spawn Count=3
    Physical Material=SurfaceType1
    Simple Max Spawn Count=3
    Simple Min Spawn Count=2
    Simple Spawn Override=2
    Spawn Count Mode=2
    Surface Type=1
   ) {module=Spawn_NDCFortImpacts002}
   &Spawn_NDCFortImpacts(
    Max Spawn Count=4
    Min Spawn Count=2
    Physical Material=SurfaceType2
    Simple Max Spawn Count=2
    Simple Min Spawn Count=1
    Simple Spawn Override=2
    Spawn Count Mode=2
    Surface Type=1
   ) {module=Spawn_NDCFortImpacts003}
   &Spawn_NDCFortImpacts(
    Max Spawn Count=7
    Min Spawn Count=3
    Physical Material=SurfaceType18
    Simple Max Spawn Count=2
    Simple Min Spawn Count=1
    Simple Spawn Override=2
    Spawn Count Mode=2
    Surface Type=1
   ) {module=Spawn_NDCFortImpacts004}
   &Spawn_NDCFortImpacts(
    Max Spawn Count=8
    Min Spawn Count=5
    Physical Material=SurfaceType24
    Simple Max Spawn Count=3
    Simple Min Spawn Count=2
    Simple Spawn Override=2
    Spawn Count Mode=2
    Surface Type=1
   ) {module=Spawn_NDCFortImpacts005}
   &Spawn_NDCFortImpacts(
    Max Spawn Count=5
    Min Spawn Count=3
    Physical Material=SurfaceType23
    Simple Max Spawn Count=2
    Simple Min Spawn Count=1
    Simple Spawn Override=2
    Spawn Count Mode=2
    Surface Type=1
   ) {module=Spawn_NDCFortImpacts006}
   &Spawn_NDCFortImpacts(
    Max Spawn Count=5
    Min Spawn Count=3
    Physical Material=SurfaceType10
    Simple Max Spawn Count=2
    Simple Min Spawn Count=1
    Simple Spawn Override=2
    Spawn Count Mode=2
    Surface Type=1
   ) {module=Spawn_NDCFortImpacts007}
   &Spawn_NDCFortImpacts(
    Max Spawn Count=5
    Min Spawn Count=3
    Physical Material=SurfaceType9
    Simple Max Spawn Count=3
    Simple Min Spawn Count=2
    Simple Spawn Override=2
    Spawn Count Mode=2
    Surface Type=1
   ) {module=Spawn_NDCFortImpacts008}
  ]
  particleSpawnScript [
   &InitializeParticleFromNDC_FortImpact(
    AmmoType Scale=1
    Distance Scale=1
    FallbackPosition:NiagaraPosition=<0 0 1>
    Scale Array=DataInterface() {propertyValues {GpuSyncMode=SyncCpuToGpu, MaxElements=4, FloatData [0.8,1,1.2,0.6,1]}}
   )
   /SetVariables(Particles.BiasedRandom=&FloatFromCurve(
     FloatCurve=DataInterface() {propertyValues {Curve {keys [
         <at 0.2 1 RCIM_Cubic>
         {arriveTangent=-3.3394, interpMode=RCIM_Cubic, leaveTangent=-3.3394, tangentMode=RCTM_User, time=0.5, value=0.5}
         <at 0.8 0 RCIM_Cubic>
        ]}}}
     CurveIndex=&RandomRangeFloat(Minimum=0, Maximum=1)
    )) {isSetParameters=true}
   &InitializeParticle(
    Color Mode=::NewEnumerator0
    Lifetime=&RandomRangeFloat(Minimum=0.5, Maximum=1)
    Mass=&Lerp_Float(A=0.6, B=1.2, Alpha=$Particles.BiasedRandom)
    Mass Mode=::NewEnumerator1
    Mesh Scale=&Multiply_VectorByFloat(
     Vector=&RandomRangeVector(Minimum=<0.9 0.4 0.5>, Maximum=<1 1 1>)
     Float=&Lerp_Float(A=0.7, B=1, Alpha=$Particles.BiasedRandom)
    )
    Mesh Scale Mode=::NewEnumerator4
    Position=$Particles.Position
    Position Mode=::NewEnumerator1
    UsePositionOffset=false
    Write Material Random=true
   )
   /SetVariables(Particles.IsSnow:NiagaraBool=&SetBoolByIntComparison(A=$Particles.PhysicalSurfaceInteger, B=23, Comparison Type=::NewEnumerator4)) {isSetParameters=true}
   &FN_PhysicalSurface(
    Dirt Mesh Count=2
    Dirt Scale:NiagaraFloat=&RandomRangeFloat(Minimum=0.33, Maximum=1)
    Dirt Velocity:NiagaraFloat=&RandomRangeFloat(Minimum=150, Maximum=400)
    Glass Mesh Count=3
    Glass Scale=0.5
    Glass Velocity:NiagaraFloat=&RandomRangeFloat(Minimum=100, Maximum=400)
    Glass Velocity Cone Angle:NiagaraFloat=&RandomRangeFloat(Minimum=20, Maximum=70)
    Grass Mass=0.25
    Grass Mesh Count=2
    Grass Scale:NiagaraFloat=&RandomRangeFloat(Minimum=0.66, Maximum=1)
    Grass Velocity:NiagaraFloat=&RandomRangeFloat(Minimum=150, Maximum=400)
    Grass Velocity Cone Angle=30
    Ice Mesh Count=2
    Ice Scale:NiagaraFloat=&MakeCustomFloatFromBool(
     Bool=$Particles.IsSnow
     True Float=&RandomRangeFloat(Minimum=0.8, Maximum=1)
     False Float=&RandomRangeFloat(Minimum=0.33, Maximum=0.5)
    )
    Ice Velocity:NiagaraFloat=&MakeCustomFloatFromBool(
     Bool=$Particles.IsSnow
     True Float=&RandomRangeFloat(Minimum=100, Maximum=300)
     False Float=&RandomRangeFloat(Minimum=100, Maximum=500)
    )
    Ice Velocity Cone Angle:NiagaraFloat=&MakeCustomFloatFromBool(
     Bool=$Particles.IsSnow
     True Float=&RandomRangeFloat(Minimum=50, Maximum=90)
     False Float=&RandomRangeFloat(Minimum=60, Maximum=120)
    )
    Ice Visibility:NiagaraInt32=&MakeCustomIntFromBool(Bool=$Particles.IsSnow, True INT=23, False INT=24)
    Metal Gravity=&RandomRangeFloat(Minimum=-1400, Maximum=-1800)
    Metal Mesh Count=2
    Metal Velocity=&RandomRangeFloat(Minimum=80, Maximum=600)
    Metal Velocity Cone Angle=30
    Physical Surface=&SelectIntBySurface(
     Default:NiagaraInt32=$Particles.PhysicalSurfaceInteger
     Surface:EPhysicalSurface=$Particles.PhysicalSurface
     Num Surfaces:NiagaraInt32=3
     Surface 0:EPhysicalSurface=SurfaceType_Default
     Value For Surface 0:NiagaraInt32=2
     Surface 1:EPhysicalSurface=SurfaceType23
     Value For Surface 1:NiagaraInt32=24
     Surface 2:EPhysicalSurface=SurfaceType9
     Value For Surface 2:NiagaraInt32=&MakeCustomIntFromBool(Bool=&RandomBool(), True INT=9, False INT=10)
     Surface 3:EPhysicalSurface=?
     Value For Surface 3:NiagaraInt32=?
    )
    Stone Mesh Count=3
    Stone Velocity=&RandomRangeFloat(Minimum=150, Maximum=400)
    Stone Velocity Cone Angle:NiagaraFloat=&RandomRangeFloat(Minimum=30, Maximum=120)
    StoneScale:NiagaraFloat=&RandomRangeFloat(Minimum=0.33, Maximum=1)
    Wood Mesh Count=2
    Wood Scale:NiagaraFloat=&RandomRangeFloat(Minimum=0.75, Maximum=1.5)
    Wood Velocity=&RandomRangeFloat(Minimum=80, Maximum=500)
    Wood Velocity Cone Angle=60
   )
   &AddVelocityInCone(
    Cone Angle=$Output.FN_PhysicalSurface.SurfaceResponseConeAngle
    Cone Axis=$Particles.ImpactSurfaceNormal
    Cone Axis Coordinate Space=::Simulation
    Module.Use Velocity Falloff On Cone Axis=false
    Velocity Strength=&Multiply_Float(A=$Output.FN_PhysicalSurface.SurfaceResponseVelocity, B=$Particles.AmmoTypeScale)
   )
   /SetVariables(Particles.Velocity=&Multiply_VectorByFloat(
     Vector=$Particles.Velocity
     Float=&Lerp_Float(A=0.5, B=2, Alpha=&ReturnNormalizedExecIndex())
    )) {isSetParameters=true}
   &AddVelocity(
    Rotation Coordinate Space=::World
    Velocity=<0 0 1>
    Velocity Speed Scale=&Multiply_Float(A=250, B=$Particles.AmmoTypeScale)
   )
   &InitialMeshOrientation(Use Rotation Vector=false)
   &AddRotationalVelocity(Rotation Rate=&RandomVector(Vector Scale=&RandomRangeFloat(Minimum=1, Maximum=2)))
   /SetVariables(Particles.GravityOverrideValue=$Output.FN_PhysicalSurface.GravityOverride) {isSetParameters=true}
   &ApplyInitialForces()
   &AddVelocity() {module=AddVelocity001}
  ]
  particleUpdateScript [
   &ParticleState()
   &Drag(
    Drag=&RandomRangeFloat(Minimum=0.5, Maximum=1)
    Rotational Drag=0.1
    Use Linear Drag=true
    Use Rotational Drag=true
   )
   &GravityForce(Gravity=&Multiply_VectorByFloat(Vector=$Particles.GravityOverrideValue, Float=2))
   &SolveForcesAndVelocity(Clamp Velocity=false, Limit Acceleration=false, Manually Enable Rotational Solver=true)
   &ScaleMeshSize(Scale Factor=&VectorFromFloat(Value=&FloatFromCurve(
      FloatCurve=DataInterface() {propertyValues {Curve {keys [<at 0.6 1 RCIM_Cubic>, {arriveTangent=-6.308, interpMode=RCIM_Cubic, leaveTangent=-6.308, tangentMode=RCTM_User, time=1, value=0}]}}}
      Scale Curve=&Lerp_Float(A=$Particles.AmmoTypeScale, B=$Particles.DistanceScale, Alpha=0.5)
     )))
   &DynamicMaterialParameters(Module.Index0Param0Write=true, Module.Index0Param1Write=true, Module.Index0Param2Write=true, Module.Index0Param3Write=true)
  ]
  rendererClasses [&NiagaraMeshRendererProperties]
  renderers [
   {rendererClass=&NiagaraMeshRendererProperties, rendererIndex=0}
   {rendererClass=&NiagaraMeshRendererProperties, rendererIndex=1}
   {rendererClass=&NiagaraMeshRendererProperties, rendererIndex=2}
   {rendererClass=&NiagaraMeshRendererProperties, rendererIndex=3}
   {rendererClass=&NiagaraMeshRendererProperties, rendererIndex=4}
   {rendererClass=&NiagaraMeshRendererProperties, rendererIndex=5}
   {rendererClass=&NiagaraMeshRendererProperties, rendererIndex=6}
  ]
 }
 Sparks {
  emitterUpdateScript [
   &EmitterState()
   &Spawn_NDCFortImpacts(
    Max Spawn Count=7
    Min Spawn Count=3
    Physical Material=SurfaceType3
    Simple Max Spawn Count=3
    Simple Min Spawn Count=1
    Simple Spawn Override=2
    Spawn Count Mode=2
    Surface Type=1
   )
  ]
  particleSpawnScript [
   &InitializeParticleFromNDC_FortImpact(
    AmmoType Scale=1
    FallbackPosition:NiagaraPosition=<0 0 1>
    Scale Array=DataInterface() {propertyValues {GpuSyncMode=SyncCpuToGpu, MaxElements=0, FloatData [0.8,1,1.25,0.6,1]}}
   )
   &InitializeParticle(
    Color Maximum=<rgba 6 1.275 0.15 1>
    Color Minimum=<rgba 2 0.425 0.05 1>
    Color Mode=::NewEnumerator2
    Lifetime=&Lerp_Float(
     A=0.133
     B=0.33
     Alpha=&FloatFromCurve(
      FloatCurve=DataInterface() {propertyValues {Curve {keys [
          {tangentMode=RCTM_User, time=0, value=0}
          {arriveTangent=1, interpMode=RCIM_Cubic, leaveTangent=1, time=0.66, value=0.15}
          <at 1 1 RCIM_Cubic>
         ]}}}
      CurveIndex=&RandomRangeFloat(Minimum=0, Maximum=1)
     )
    )
    Mass Mode=::NewEnumerator1
    Position=$Particles.Position
    Position Mode=::NewEnumerator1
    Sprite Size Mode=::NewEnumerator1
    Uniform Sprite Size Max=2.5
    Uniform Sprite Size Min=1.5
    UsePositionOffset=false
    Write Material Random=true
   )
   &AddVelocity(
    Rotation Coordinate Space=::World
    Velocity=<0 0 1>
    Velocity Speed Scale=&Multiply_Float(A=200, B=$Particles.AmmoTypeScale)
   )
   &AddVelocityInCone(
    Cone Angle=&RandomRangeFloat(Minimum=150, Maximum=170)
    Cone Axis=$Particles.ImpactSurfaceNormal
    Module.Use Velocity Falloff On Cone Axis=false
    Velocity Strength=&Multiply_Float(A=&RandomRangeFloat(Minimum=300, Maximum=800), B=$Particles.AmmoTypeScale)
   )
  ]
  particleUpdateScript [
   &ParticleState()
   &Drag(Drag=&RandomRangeFloat(Minimum=0.25, Maximum=3), Use Linear Drag=true, Use Rotational Drag=false)
   &GravityForce()
   &SolveForcesAndVelocity(Clamp Velocity=false, Limit Acceleration=false)
   &ScaleSpriteSizeBySpeed(
    Max Scale Factor=<0.66 12>
    Sample Scale Factor By Curve=true
    Scale Factor Curve=DataInterface() {propertyValues {Curve {keys [<at 0 0 RCIM_Cubic>, {arriveTangent=1.6365, tangentMode=RCTM_User, time=1, value=1}]}}}
    Velocity Threshold=800
   )
   &ScaleColor(
    Scale RGB=&Lerp_Vector(
     A=<0.2 0 0>
     B=<1 1 1>
     Alpha=&RemapRange(
      Input Value To Remap=&VectorLength(Vector=$Particles.Velocity)
      Input Min=0
      Input Max=250
      Output Min=0
      Output Max=1
     )
    )
    ScaleA=true
    ScaleRGB=true
   )
  ]
  rendererClasses [&NiagaraSpriteRendererProperties]
  renderers [{rendererClass=&NiagaraSpriteRendererProperties}]
 }
 Sparks_GPU {
  bEnabled=false
  simTarget=GPUComputeSim
  emitterUpdateScript [
   &EmitterState()
   &Spawn_NDCFortImpacts(
    Distance Culling=1
    LOD Distance=6000
    Max Spawn Count=7
    Min Spawn Count=3
    Physical Material=SurfaceType3
    Simple Max Spawn Count=3
    Simple Min Spawn Count=1
    Simple Spawn Override=2
    Spawn Count Mode=2
    Surface Type=1
   )
   &Spawn_NDCFortImpacts(
    Distance Culling=1
    LOD Distance=6000
    Max Spawn Count=2
    Min Spawn Count=0
    Physical Material=SurfaceType2
    Simple Max Spawn Count=1
    Simple Spawn Override=2
    Spawn Count Mode=2
    Surface Type=1
   ) {module=Spawn_NDCFortImpacts001}
  ]
  particleSpawnScript [
   &InitializeParticleFromNDC_FortImpact(
    AmmoType Scale=1
    FallbackPosition:NiagaraPosition=<0 0 1>
    Scale Array=DataInterface() {propertyValues {GpuSyncMode=SyncCpuToGpu, MaxElements=0, FloatData [0.8,1,1.25,0.6,1]}}
   )
   &InitializeParticle(
    Color Maximum=<rgba 6 1.275 0.15 1>
    Color Minimum=<rgba 2 0.425 0.05 1>
    Color Mode=::NewEnumerator2
    Lifetime=&Lerp_Float(
     A=0.5
     B=1.5
     Alpha=&FloatFromCurve(
      FloatCurve=DataInterface() {propertyValues {Curve {keys [
          {tangentMode=RCTM_User, time=0, value=0}
          {arriveTangent=1, interpMode=RCIM_Cubic, leaveTangent=1, time=0.66, value=0.15}
          {tangentMode=RCTM_User, time=1, value=1}
         ]}}}
      CurveIndex=&RandomRangeFloat(Minimum=0, Maximum=1)
     )
    )
    Mass Mode=::NewEnumerator1
    Position=$Particles.Position
    Position Mode=::NewEnumerator1
    Sprite Size Mode=::NewEnumerator1
    Uniform Sprite Size Max=4
    Uniform Sprite Size Min=2.5
    UsePositionOffset=false
    Write Material Random=true
   )
   &AddVelocity(
    Rotation Coordinate Space=::World
    Velocity=<0 0 1>
    Velocity Speed Scale=&Multiply_Float(A=200, B=$Particles.AmmoTypeScale)
   )
   &AddVelocityInCone(
    Cone Angle=&RandomRangeFloat(Minimum=160, Maximum=170)
    Cone Axis=$Particles.ImpactSurfaceNormal
    Module.Use Velocity Falloff On Cone Axis=false
    Velocity Strength=&Multiply_Float(
     A=&Multiply_Float(A=&RandomRangeFloat(Minimum=300, Maximum=800), B=$Particles.AmmoTypeScale)
     B=&SelectFloatBySurface(
      Default:NiagaraFloat=1
      Surface:EPhysicalSurface=$Particles.PhysicalSurface
      Num Surfaces:NiagaraInt32=1
      Surface 0:EPhysicalSurface=SurfaceType2
      Value For Surface 0:NiagaraFloat=0.66
      Surface 1:EPhysicalSurface=SurfaceType2
      Value For Surface 1:NiagaraFloat=?
      Surface 2:EPhysicalSurface=?
      Value For Surface 2:NiagaraFloat=?
      Surface 3:EPhysicalSurface=?
      Value For Surface 3:NiagaraFloat=?
     )
    )
   )
  ]
  particleUpdateScript [
   &ParticleState()
   &Drag(Drag=&RandomRangeFloat(Minimum=0.25, Maximum=2), Use Linear Drag=true, Use Rotational Drag=false)
   &GravityForce(Gravity=&Multiply_VectorByFloat(Vector=<0 0 -980>, Float=1.5))
   &Collision(
    Age Colliding Particles=false
    Friction:NiagaraFloat=&RandomRangeFloat(Minimum=0.25, Maximum=0.5)
    Friction During a Bounce:NiagaraFloat=&RandomRangeFloat(Minimum=0, Maximum=0.1)
    Kill Offscreen Particles=true
    Maximum time offscreen=0.5
    Module.Control Rest State Via Surface Normal=false
    Module.EnableMaxCollisionCount=false
    Module.Randomize Collision Normal Vector=false
    Restitution=&RandomRangeFloat(Minimum=0.25, Maximum=0.4)
   ) {enabled=false}
   &SolveForcesAndVelocity(Clamp Velocity=false, Limit Acceleration=false)
   &ScaleSpriteSizeBySpeed(
    Max Scale Factor=<0.66 12>
    Sample Scale Factor By Curve=true
    Scale Factor Curve=DataInterface() {propertyValues {Curve {keys [<at 0 0 RCIM_Cubic>, {arriveTangent=1.6365, tangentMode=RCTM_User, time=1, value=1}]}}}
    Velocity Threshold=800
   )
   &ScaleColor(
    Scale RGB=&Lerp_Vector(
     A=<0.2 0 0>
     B=<1 1 1>
     Alpha=&RemapRange(
      Input Value To Remap=&VectorLength(Vector=$Particles.Velocity)
      Input Min=0
      Input Max=250
      Output Min=0
      Output Max=&FloatFromCurve(FloatCurve=DataInterface() {propertyValues {Curve {keys [{tangentMode=RCTM_User, time=0.66, value=1}, {tangentMode=RCTM_User, time=1, value=0}]}}})
     )
    )
    ScaleA=true
    ScaleRGB=true
   )
  ]
  rendererClasses [&NiagaraSpriteRendererProperties]
  renderers [{rendererClass=&NiagaraSpriteRendererProperties}]
 }
 Spike_Streaks {
  bEnabled=false
  emitterUpdateScript [
   &EmitterState()
   &Spawn_NDCFortImpacts(Max Spawn Count=5, Min Spawn Count=3, Simple Max Spawn Count=2, Simple Min Spawn Count=1, Simple Spawn Override=2, Spawn Count Mode=2)
  ]
  particleSpawnScript [
   &InitializeParticleFromNDC_FortImpact(
    AmmoType Scale=1
    FallbackPosition:NiagaraPosition=<0 0 1>
    Scale Array=DataInterface() {propertyValues {GpuSyncMode=SyncCpuToGpu, MaxElements=4, FloatData [0.5,0.8,1.25,0.35,1]}}
   )
   &InitializeParticle(
    Color Maximum=<rgba 8 4 0.4 1>
    Color Minimum=<rgba 4 2 0.2 1>
    Color Mode=::NewEnumerator2
    Lifetime Max=0.1
    Lifetime Min=0.06
    Lifetime Mode=::NewEnumerator1
    Mesh Scale=&Multiply_VectorByFloat(
     Vector=&RandomRangeVector(Minimum=<1 0.15 0.15>, Maximum=<2 0.25 0.25>)
     Float=$Particles.AmmoTypeScale
    )
    Mesh Scale Mode=::NewEnumerator4
    Position=$Particles.Position
    Position Mode=::NewEnumerator1
    UsePositionOffset=false
    Write Material Random=true
   )
   &InitialMeshOrientation(
    Mesh Orientation Mode=::NewEnumerator2
    Orientation Vector=$Particles.ImpactSurfaceNormal
    Rotation=&RandomRangeVector(Minimum=<0 -0.2 -0.2>, Maximum=<1 0.2 0.2>)
    Use Rotation Vector=true
   )
   &CameraOffset(Camera Offset Amount=10)
  ]
  particleUpdateScript [
   &ParticleState()
   &Color(
    Color=&MakeLinearColorFromVectorAndFloat(
     "Vector (RGB)"=&MakeVectorFromLinearColorRGB(Color=&ColorFromCurve(ColorCurve=DataInterface() {propertyValues {
         AlphaCurve {keys []}
         BlueCurve {keys [
           {arriveTangent=-4.1816, interpMode=RCIM_Cubic, leaveTangent=-4.1816, time=0, value=4.1781}
           {arriveTangent=-4.1816, interpMode=RCIM_Cubic, leaveTangent=-4.1816, time=0.9977, value=0.0059}
          ]}
         GreenCurve {keys [
           {arriveTangent=-7.3723, interpMode=RCIM_Cubic, leaveTangent=-7.3723, time=0, value=7.3615}
           {arriveTangent=-7.3723, interpMode=RCIM_Cubic, leaveTangent=-7.3723, time=0.9977, value=0.0059}
          ]}
         RedCurve {keys [
           {arriveTangent=-14.4701, interpMode=RCIM_Cubic, leaveTangent=-14.4701, time=0, value=15}
           {arriveTangent=-14.4701, interpMode=RCIM_Cubic, leaveTangent=-14.4701, time=0.9977, value=0.5625}
          ]}
        }}))
     "Float (Alpha)"=&FloatFromCurve(FloatCurve=DataInterface() {propertyValues {Curve {keys [{interpMode=RCIM_Cubic, tangentMode=RCTM_User, time=0, value=1}, {interpMode=RCIM_Cubic, tangentMode=RCTM_User, time=1, value=0}]}}})
    )
    Scale Color=&VectorFromFloat(Value=&FloatFromCurve(FloatCurve=DataInterface() {propertyValues {Curve {keys [{interpMode=RCIM_Cubic, tangentMode=RCTM_User, time=0, value=1}, {interpMode=RCIM_Cubic, tangentMode=RCTM_User, time=1, value=0}]}}}))
   ) {enabled=false}
   &ScaleColor(
    Scale RGB=&Lerp_Vector(
     A=<0.25 0 0>
     B=<1 1 1>
     Alpha=&FloatFromCurve(FloatCurve=DataInterface() {propertyValues {Curve {keys [<at 0.33 1 RCIM_Cubic>,<at 1 0>]}}})
    )
    ScaleA=true
    ScaleRGB=true
   )
   &ScaleMeshSize(Scale Factor=&VectorFromCurve(
     VectorCurve=DataInterface() {propertyValues {XCurve {keys [<at 0 0.4>,<at 1 1.2>]}, YCurve {keys [<at 0 1>,<at 1 0>]}, ZCurve {keys [<at 0 1>,<at 1 0>]}}}
     Scale Curve=&VectorFromFloat(Value=&RemapRange(Input Value To Remap=$Particles.LODDistance, Input Min=100, Input Max=8000, Output Min=1, Output Max=3))
    ))
  ]
  rendererClasses [&NiagaraMeshRendererProperties]
  renderers [{rendererClass=&NiagaraMeshRendererProperties, rendererIndex=0}]
 }
 Streaks {
  emitterUpdateScript [
   &EmitterState()
   &Spawn_NDCFortImpacts(Max Spawn Count=6, Min Spawn Count=4, Simple Max Spawn Count=3, Simple Min Spawn Count=2, Simple Spawn Override=2, Spawn Count Mode=2)
  ]
  particleSpawnScript [
   &InitializeParticleFromNDC_FortImpact(
    AmmoType Scale=1
    FallbackPosition:NiagaraPosition=<0 0 1>
    Scale Array=DataInterface() {propertyValues {GpuSyncMode=SyncCpuToGpu, MaxElements=0, FloatData [0.5,0.8,1.25,0.4,1]}}
   )
   &InitializeParticle(
    Color Maximum=<rgba 8 0.2 0.33 1>
    Color Minimum=<rgba 4 0.1 0.75 1>
    Color Mode=::NewEnumerator2
    Lifetime=&Multiply_Float(
     A=&UniformRangedFloat(Minimum=0.08, Maximum=0.21)
     B=&SelectFloatFromArray(Direct Array Index=$Particles.AmmoType, Float Selection Array=DataInterface() {propertyValues {FloatData [0.6,0.7,1,0.5,0.8]}})
    )
    Mass Mode=::NewEnumerator1
    Position=$Particles.Position
    Position Mode=::NewEnumerator1
    Sprite Size=&RandomRangeVector2D(Minimum=<1 5>, Maximum=<3 30>)
    Sprite Size Mode=::NewEnumerator4
    UsePositionOffset=false
    Write Material Random=true
   )
   &AddVelocityInCone(
    Cone Angle=&RandomRangeFloat(Minimum=75, Maximum=100)
    Cone Axis=$Particles.ImpactSurfaceNormal
    Cone Axis Coordinate Space=::World
    Module.Use Velocity Falloff On Cone Axis=false
    Velocity Strength=&Multiply_Float(
     A=&RandomRangeFloat(
      Minimum=&SelectFloatBySurface(
       Default:NiagaraFloat=150
       Surface:EPhysicalSurface=$Particles.PhysicalSurface
       Num Surfaces:NiagaraInt32=1
       Surface 0:EPhysicalSurface=SurfaceType1
       Value For Surface 0:NiagaraFloat=200
       Surface 1:EPhysicalSurface=?
       Value For Surface 1:NiagaraFloat=?
       Surface 2:EPhysicalSurface=?
       Value For Surface 2:NiagaraFloat=?
       Surface 3:EPhysicalSurface=?
       Value For Surface 3:NiagaraFloat=?
      )
      Maximum=&SelectFloatBySurface(
       Default:NiagaraFloat=800
       Surface:EPhysicalSurface=$Particles.PhysicalSurface
       Num Surfaces:NiagaraInt32=1
       Surface 0:EPhysicalSurface=SurfaceType1
       Value For Surface 0:NiagaraFloat=900
       Surface 1:EPhysicalSurface=?
       Value For Surface 1:NiagaraFloat=?
       Surface 2:EPhysicalSurface=?
       Value For Surface 2:NiagaraFloat=?
       Surface 3:EPhysicalSurface=?
       Value For Surface 3:NiagaraFloat=?
      )
     )
     B=$Particles.AmmoTypeScale
    )
   )
  ]
  particleUpdateScript [
   &ParticleState()
   &Drag(Drag=&RandomRangeFloat(Minimum=2, Maximum=2.5), Use Linear Drag=true, Use Rotational Drag=false)
   &SolveForcesAndVelocity(Clamp Velocity=false, Limit Acceleration=false)
   &Color(
    Color=<rgba 1 0.3636 0.01 1>
    Scale Color=&VectorFromCurve(VectorCurve=DataInterface() {propertyValues {XCurve {keys [<at 0.7 12>,<at 1 1>]}, YCurve {keys [<at 0.7 12>,<at 1 1>]}, ZCurve {keys [<at 0.7 12>,<at 1 1>]}}})
   ) {enabled=false}
   &ScaleSpriteSize(
    Non-Uniform Curve Scale=&Vector2DFromFloat(Value=&RemapRange(Input Value To Remap=$Particles.LODDistance, Input Min=100, Input Max=8000, Output Min=1, Output Max=3))
    Non-Uniform Curve Sprite Scale=DataInterface() {propertyValues {
      XCurve {keys [{tangentMode=RCTM_User, time=0, value=1}, {tangentMode=RCTM_User, time=1, value=0.2}]}
      YCurve {keys [{tangentMode=RCTM_User, time=0, value=1.5}, {tangentMode=RCTM_User, time=1, value=0.7}]}
     }}
    Scale Sprite Size Mode=::NewEnumerator3
   )
   &ScaleColor(
    Scale RGB=&Lerp_Vector(
     A=<0.25 0 0>
     B=<1 1 1>
     Alpha=&FloatFromCurve(FloatCurve=DataInterface() {propertyValues {Curve {keys [<at 0.33 1 RCIM_Cubic>,<at 1 0>]}}})
    )
    ScaleA=true
    ScaleRGB=true
   )
  ]
  rendererClasses [&NiagaraSpriteRendererProperties]
  renderers [{rendererClass=&NiagaraSpriteRendererProperties}]
 }
}
summary {
 systemName=NS_NDC_Handler_Impacts_Spire
 emitters [
  {
   emitterName=Flash
   rendererClasses [
    &NiagaraMeshRendererProperties
    &NiagaraSpriteRendererProperties
    &NiagaraLightRendererProperties
   ]
  }
  {bEnabled=false, emitterName=Light, rendererClasses [&NiagaraLightRendererProperties]}
  {emitterName=Streaks, rendererClasses [&NiagaraSpriteRendererProperties]}
  {emitterName=Cloud, rendererClasses [&NiagaraSpriteRendererProperties]}
  {bEnabled=false, emitterName=Spike_Streaks, rendererClasses [&NiagaraMeshRendererProperties]}
  {emitterName=Airflow, rendererClasses [&NiagaraMeshRendererProperties]}
  {emitterName=Sparks, rendererClasses [&NiagaraSpriteRendererProperties]}
  {emitterName=NE_FN_PhysMatDebris, rendererClasses [&NiagaraMeshRendererProperties]}
  {bEnabled=false, emitterName=Debug, rendererClasses [&NiagaraSpriteRendererProperties]}
  {emitterName=Decal, rendererClasses [&NiagaraDecalRendererProperties]}
  {bEnabled=false, emitterName=Sparks_GPU, simTarget=GPUComputeSim, rendererClasses [&NiagaraSpriteRendererProperties]}
 ]
 userVariables []
}
