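# Parent material for the Spire LMG weapon, adding a thin-film anodizing sheen to the shared 2023 character material stack.
# A Fresnel with adjustable exponent samples a curve-atlas color, which is scaled by an intensity parameter and masked out by the GoldMask channel of the Tech Art Mask.
# That ThinFilm result is added to Base Color between the character PreFX and PostFX functions, using material attributes.
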Material /CherrySmokeWeaponGameplay/Gameplay/SpireGun/Materials/M_SpireLMG_Parent
graph {
 properties {
  bUseMaterialAttributes=true
  bUsedWithClothing=true
  bUsedWithMeshDeformer=true
  bUsedWithMorphTargets=true
  bUsedWithNanite=true
  bUsedWithNiagaraMeshParticles=true
  bUsedWithSkeletalMesh=true
  pixelDepthOffsetMode=PDOM_Legacy
  shadingModel=MSM_FromMaterialExpression
  usedShadingModels=Default Lit | Subsurface | Clear Coat | Subsurface Profile | Cloth
 }
 nodes [
  n1 = MaterialFunctionCall() {materialFunction=&MF_CharacterParent_2023_PreFX}
  n2 = GetMaterialAttributes(None=@n1.Result) {attributeGetTypes [#1]}
  n3 = NamedRerouteUsage()
  MaterialFunctionCall(
   Material=SetMaterialAttributes!(MaterialAttributes=@n2.MaterialAttributes, Base Color=Add(A=@n2.Base Color, B=@n3)) {attributeSetTypes [#1]}
   M=@n1.M
   Pre Skinned Local Position=@n1.M
   Distance Blend=@n1.M
   Fresnel=@n1.M
  ) {materialFunction=&MF_CharacterParent_2023_PostFX}
  n4 = CurveAtlasRowParameter(CurveTime=Fresnel(ExponentIn=ScalarParameter() {defaultValue=1, group=Spire, parameterName=Anodizing_Exponent, sliderMax=32}) {baseReflectFraction=0, exponent=3}) {
   atlas=&/CA_SpireCurves
   curve=&/CC_SpireThinFilm
   group=01_Spire
   parameterName=Anodizing Color
  }
  n5 = ScalarParameter() {defaultValue=0.03, group=Spire, parameterName=Anodizing, sliderMax=1}
  n6 = ChannelMaskParameter(None=@n1.Tech Art Mask) {group=Spire, parameterName=GoldMask}
  NamedRerouteDeclaration(None=Multiply(A=Multiply(A=@n4, B=@n5), B=OneMinus(None=@n6.R))) {name=ThinFilm, nodeColor=<rgba 0 1 0.5606 1>}
 ]
}
