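# Control Rig for the Heavy Spire gun, procedurally building animation controls over its weapon skeleton.
# Construction stacks root, offset and handle controls, generates one per bone under handle into an array, then styles hand, trigger, ammo, bullet and barrel shapes.
# Forward solve parent-constrains the three main bones and drives the rest through an FK chain.
# Backwards solve copies bone transforms onto controls, relying on controls sharing bone names.
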ControlRigBlueprint /CherrySmokeWeaponGameplay/Gameplay/SpireGun/Meshes/CR_HeavySpire
graphs [
 New Function_ContainedGraph {nodes [
   RigVMFunctionEntryNode()
   > RigVMFunction_Sequence = RigVMAggregateNode()
   n1 = RigVMVariableNode() {variable=Root Ctrl Name}
   Control Stack at Item = RigVMFunctionReferenceNode(
    Add Bottom Null=False
    Add Null=true
    Add Secondary Control=False
    Item=(Name="root",Type=Bone)
    Null Suffix=_null
    Orient to World=false
    Parent=(Name="None",Type=None)
    Secondary Settings="(DisplayName=\"None\",Shape=(bVisible=True,Name=\"Default\",Color=(R=1,G=0,B=0,A=1),Transform=(Rotation=(X=0,Y=0,Z=0,W=1),Translation=(X=0,Y=0,Z=0),Scale3D=(X=1,Y=1,Z=1))),Proxy=(bIsProxy=False,ShapeVisibility=BasedOnSelection),InitialSpace=LocalSpace,bUsePref ...(truncated)"
    Tag Controls=true
    Zero Controls=true
    ExecuteContext=@RigVMFunction_Sequence.A
    Item.Name=@n1.Value
    Parent.Name=@n1.Value
    Settings=RigVMVariableNode() {variable=Root Ctrl Setting}.Value
   )
   > RigVMVariableNode(Value=@"Control Stack at Item".Control) {variable=Root Ctrl}
   n2 = RigVMVariableNode() {variable=Offset Ctrl Name}
   > Control Stack at Item_1 = RigVMFunctionReferenceNode(
    Add Bottom Null=False
    Add Null=false
    Add Secondary Control=False
    Item=(Name="offset",Type=Bone)
    Null Suffix=_null
    Orient to World=false
    Secondary Settings="(DisplayName=\"None\",Shape=(bVisible=True,Name=\"Default\",Color=(R=1,G=0,B=0,A=1),Transform=(Rotation=(X=0,Y=0,Z=0,W=1),Translation=(X=0,Y=0,Z=0),Scale3D=(X=1,Y=1,Z=1))),Proxy=(bIsProxy=False,ShapeVisibility=BasedOnSelection),InitialSpace=LocalSpace,bUsePref ...(truncated)"
    Tag Controls=true
    Zero Controls=true
    Parent=@"Control Stack at Item".Control
    Item.Name=@n2.Value
    Name=@n2.Value
    Settings=RigVMVariableNode() {variable=Offset Ctrl Setting}.Value
   )
   > RigVMVariableNode(Value=@"Control Stack at Item_1".Control) {variable=Offset Ctrl}
   n3 = RigVMVariableNode() {variable=Handle Bone Name}
   > Control Stack at Item_2 = RigVMFunctionReferenceNode(
    Add Bottom Null=False
    Add Null=false
    Add Secondary Control=False
    Item=(Name="handle",Type=Bone)
    Null Suffix=_null
    Orient to World=false
    Secondary Settings="(DisplayName=\"None\",Shape=(bVisible=True,Name=\"Default\",Color=(R=1,G=0,B=0,A=1),Transform=(Rotation=(X=0,Y=0,Z=0,W=1),Translation=(X=0,Y=0,Z=0),Scale3D=(X=1,Y=1,Z=1))),Proxy=(bIsProxy=False,ShapeVisibility=BasedOnSelection),InitialSpace=LocalSpace,bUsePref ...(truncated)"
    Tag Controls=true
    Zero Controls=true
    Parent=@"Control Stack at Item_1".Control
    Item.Name=@n3.Value
    Name=@n3.Value
    Settings=RigVMVariableNode() {variable=Handle Ctrl Setting}.Value
   )
   > RigVMVariableNode(Value=@"Control Stack at Item_2".Control) {variable=Handle Ctrl}
   RigVMFunctionReturnNode(ExecuteContext=@RigVMFunction_Sequence.D)
  ]}
 New Function_ContainedGraph {nodes [
   RigVMFunctionEntryNode()
   > RigVMFunction_Sequence = RigVMAggregateNode()
   &RigUnit_ParentConstraint(
    AdvancedSettings=(InterpolationType=Average,RotationOrderForFilter=XZY)
    Child=(Type=Bone,Name="root")
    Parents=((Item=(Type=Control,Name="root"),Weight=1))
    Weight=1
    bMaintainOffset=true
    ExecutePin=@RigVMFunction_Sequence.A
    Parents.0.Item.Name=RigVMVariableNode() {variable=Root Ctrl Name}.Value
   )
   > &RigUnit_ParentConstraint(
    AdvancedSettings=(InterpolationType=Average,RotationOrderForFilter=XZY)
    Child=(Type=Bone,Name="offset")
    Parents=((Item=(Type=Control,Name="offset"),Weight=1))
    Weight=1
    bMaintainOffset=true
    Parents.0.Item.Name=RigVMVariableNode() {variable=Offset Ctrl Name}.Value
   )
   > &RigUnit_ParentConstraint(
    AdvancedSettings=(InterpolationType=Average,RotationOrderForFilter=XZY)
    Child=(Type=Bone,Name="handle")
    Parents=((Item=(Type=Control,Name="handle"),Weight=1))
    Weight=1
    bMaintainOffset=true
    Parents.0.Item.Name=RigVMVariableNode() {variable=Handle Bone Name}.Value
   )
   FKChain = RigVMFunctionReferenceNode(
    Rotation Filter=(bX=True,bY=True,bZ=True)
    Translation Filter=(bX=True,bY=True,bZ=True)
    Bones=/Script/ControlRig.RigUnit_CollectionChildrenArray(
     Parent=(Type=Bone,Name="handle")
     TypeToSearch=Bone
     bDefaultChildren=True
     bIncludeParent=false
     bRecursive=true
     Parent.Name=RigVMVariableNode() {variable=Handle Bone Name}.Value
    ) {is=struct RigVMUnitNode}.Items
    ExecuteContext=@RigVMFunction_Sequence.B
    Controls=RigVMVariableNode() {variable=Children Ctrls}.Value
   )
   RigVMFunctionReturnNode(ExecuteContext=@RigVMFunction_Sequence.D)
  ]}
 New Function_ContainedGraph {nodes [
   RigVMFunctionEntryNode()
   > RigVMFunction_Sequence = RigVMAggregateNode()
   n1 = &RigUnit_ItemArray(Items=((Type=Bone,Name="barrel_l_01"),(Type=Bone,Name="barrel_l_02"),(Type=Bone,Name="barrel_l_03"),(Type=BONE,Name="barrel_l_04"),(Type=BONE,Name="barrel_l_05")))
   n2 = RigVMDispatchNode(Array=@n1.Items, ExecuteContext=@RigVMFunction_Sequence.D) {template="DISPATCH_RigVMDispatch_ArrayIterator(in Array,out Element,out Index,out Count,out Ratio)"}
   > Set Control Shape_1_1_1_5 = RigVMFunctionReferenceNode(
    Control Settings=(bVisible=True,Name="Arrow_Solid",Color=(R=0,G=0,B=1,A=1),Transform=(Rotation=(X=0,Y=0,Z=-0.3827,W=0.9239),Translation=(X=-5,Y=0,Z=0),Scale3D=(X=0.2,Y=0.2,Z=1)))
    Control Name=@n2.Element.Name
   )
   RigUnit_ItemArray = &RigUnit_ItemArray(Items=((Type=Bone,Name="barrel_r_01"),(Type=Bone,Name="barrel_r_02"),(Type=Bone,Name="barrel_r_03"),(Type=BONE,Name="barrel_r_04"),(Type=BONE,Name="barrel_r_05")))
   n3 = RigVMDispatchNode(ExecuteContext=@n2.Completed, Array=@RigUnit_ItemArray.Items) {template="DISPATCH_RigVMDispatch_ArrayIterator(in Array,out Element,out Index,out Count,out Ratio)"}
   > Set Control Shape_1_1_1_6 = RigVMFunctionReferenceNode(
    Control Settings=(bVisible=True,Name="Arrow_Solid",Color=(R=1,G=0,B=0,A=1),Transform=(Rotation=(X=0,Y=0,Z=-0.3827,W=0.9239),Translation=(X=-5,Y=0,Z=0),Scale3D=(X=0.2,Y=0.2,Z=1)))
    Control Name=@n3.Element.Name
   )
   RigVMFunctionReturnNode(ExecuteContext=@RigVMFunction_Sequence.F)
   DISPATCH_RigVMDispatch_ArrayIterator = RigVMDispatchNode(
    Array=&RigUnit_CollectionChildrenArray(
     Parent=(Type=Bone,Name="handle")
     TypeToSearch=Bone
     bDefaultChildren=True
     bIncludeParent=false
     bRecursive=true
     Parent.Name=RigVMVariableNode() {variable=Handle Bone Name}.Value
    ).Items
    ExecuteContext=@RigVMFunction_Sequence.A
   ) {template="DISPATCH_RigVMDispatch_ArrayIterator(in Array,out Element,out Index,out Count,out Ratio)"}
   > Control Stack at Item_3 = RigVMFunctionReferenceNode(
    Add Bottom Null=False
    Add Null=false
    Add Secondary Control=false
    Null Suffix=_null
    Orient to World=false
    Parent=(Name="handle_ctrl",Type=Control)
    Secondary Settings="(DisplayName=\"None\",Shape=(bVisible=true,Name=\"Default\",Color=(R=0,G=1,B=0,A=1),Transform=(Rotation=(X=0,Y=0,Z=0,W=1),Translation=(X=0,Y=0,Z=0),Scale3D=(X=0.1,Y=0.1,Z=0.1))),Proxy=(bIsProxy=False,ShapeVisibility=BasedOnSelection),InitialSpace=LocalSpace,bU ...(truncated)"
    Settings="(DisplayName=\"None\",Shape=(bVisible=True,Name=\"Sphere_Thick\",Color=(R=1,G=0,B=0,A=1),Transform=(Rotation=(X=0,Y=0,Z=0,W=1),Translation=(X=0,Y=0,Z=0),Scale3D=(X=0.5,Y=0.5,Z=0.5))),Proxy=(bIsProxy=False,ShapeVisibility=BasedOnSelection),InitialSpace=LocalSpa ...(truncated)"
    Tag Controls=true
    Zero Controls=true
    Item=@DISPATCH_RigVMDispatch_ArrayIterator.Element
    Parent.Name=/Script/ControlRig.RigUnit_HierarchyGetParent(Child=(Type=Bone,Name="None"), bDefaultParent=True, Child.Name=@DISPATCH_RigVMDispatch_ArrayIterator.Element.Name) {is=struct RigVMUnitNode}.Parent.Name
   )
   > DISPATCH_RigVMDispatch_ArrayAdd = RigVMDispatchNode(Element=@"Control Stack at Item_3".Control, Array=RigVMVariableNode() {variable=Children Ctrls}.Value) {template="DISPATCH_RigVMDispatch_ArrayAdd(io Array,in Element,out Index)"}
   Set Control Shape_1_1_1_1 = RigVMFunctionReferenceNode(
    Control Settings=RigVMVariableNode() {variable=Left Hand Ctrl Settings}.Value
    Control Name=RigVMVariableNode() {variable="Left Hand  Ctrl Name"}.Value
    ExecuteContext=@RigVMFunction_Sequence.C
   )
   > Set Control Shape_1_1_1_1_1 = RigVMFunctionReferenceNode(
    Control Name=RigVMVariableNode() {variable=Right Hand Ctrl Name}.Value
    Control Settings=RigVMVariableNode() {variable=Right Hand Ctrl Settings}.Value
   )
   > Set Control Shape_1_1_1_2 = RigVMFunctionReferenceNode(
    Control Name=trigger
    Control Settings=(bVisible=True,Name="Arrow_Thick",Color=(R=1,G=0,B=0,A=1),Transform=(Rotation=(X=-0.5,Y=-0.5,Z=-0.5,W=0.5),Translation=(X=1,Y=0,Z=-2),Scale3D=(X=0.3,Y=0.3,Z=0.3)))
   )
   > Set Control Shape_1_1_1_3 = RigVMFunctionReferenceNode(
    Control Name=ammo_compartment
    Control Settings=(bVisible=True,Name="HalfCircle_Thick",Color=(R=1,G=1,B=0,A=1),Transform=(Rotation=(X=0,Y=-0.7071,Z=0.7071,W=0),Translation=(X=1,Y=0,Z=10),Scale3D=(X=0.75,Y=0.5,Z=1)))
   )
   > Set Control Shape_1_1_1_4 = RigVMFunctionReferenceNode(
    Control Name=bullet
    Control Settings=(bVisible=True,Name="Box_Thick",Color=(R=1,G=0,B=1,A=1),Transform=(Rotation=(X=0,Y=-0.7071,Z=0.7071,W=0),Translation=(X=3.5,Y=0,Z=0),Scale3D=(X=1,Y=0.5,Z=0.5)))
   )
  ]}
 New Function_ContainedGraph {nodes [
   n1 = RigVMFunctionEntryNode()
   n2 = &RigVMFunction_ControlFlowBranch(
    ExecuteContext=@n1.ExecuteContext
    Condition=/Script/ControlRig.RigUnit_ItemExists(Item=(Type=Control,Name="trigger_ctrl"), Item.Name=RigVMVariableNode() {variable=Control Name}.Value) {is=struct RigVMUnitNode}.Exists
   )
   RigVMFunctionReturnNode(ExecuteContext=@n2.Completed)
   /Script/ControlRig.RigUnit_HierarchySetShapeSettings(
    Item=(Type=Control,Name="trigger_ctrl")
    ExecutePin=@n2.True
    Item.Name=RigVMVariableNode() {variable=Control Name}.Value
    Settings=RigVMVariableNode() {variable=Control Settings}.Value
   ) {is=struct RigVMUnitNode}
  ]}
 New Function_ContainedGraph {nodes [
   n1 = RigVMFunctionEntryNode()
   > DISPATCH_RigVMDispatch_ArrayIterator = RigVMDispatchNode(Array=/Script/ControlRig.RigUnit_CollectionNameSearchArray(TypeToSearch=Control, PartialName=@n1.PartialName) {is=struct RigVMUnitNode}.Items) {template="DISPATCH_RigVMDispatch_ArrayIterator(in Array,out Element,out Index,out Count,out Ratio)"}
   SetControlColor = /Script/ControlRig.RigUnit_SetControlColor(
    ExecutePin=@DISPATCH_RigVMDispatch_ArrayIterator.Completed
    Control=@DISPATCH_RigVMDispatch_ArrayIterator.Element.Name
    Color=RigVMVariableNode() {variable=Color}.Value
   ) {is=struct RigVMUnitNode, template="SetControlColor::Execute(in Control,in Color)"}
   > RigVMFunctionReturnNode()
  ]}
 New Function_ContainedGraph {nodes [
   RigVMFunctionEntryNode()
   > RigVMFunction_Sequence = RigVMAggregateNode()
   &RigUnit_HierarchySetShapeSettings(
    Item=(Type=Control,Name="attach_mag")
    Settings=(bVisible=True,Name="Box_Thick",Color=(R=1,G=0,B=0,A=1),Transform=(Rotation=(X=0,Y=0,Z=0,W=1),Translation=(X=0,Y=0,Z=0),Scale3D=(X=1,Y=1,Z=1)))
    ExecutePin=@RigVMFunction_Sequence.B
    Settings.Color=RigVMVariableNode() {variable=New Control Color}.Value
    Settings.Transform.Translation=RigVMVariableNode() {variable=Translation}.Value
    Settings.Transform.Scale3D=RigVMVariableNode() {variable=Scale}.Value
    Settings.Transform.Rotation=RigVMVariableNode() {variable=Rotation}.Value
    Settings.Name=RigVMVariableNode() {variable=Shape Name}.Value
    Item.Name=RigVMVariableNode() {variable=Control Name}.Value
    Settings.bVisible=RigVMVariableNode() {variable=Visibility}.Value
   )
   RigVMFunctionReturnNode(ExecuteContext=@RigVMFunction_Sequence.C)
   n1 = /Script/ControlRig.RigUnit_HierarchyGetShapeSettings(Item=(Type=Control,Name="attach_mag"), Item.Name=RigVMVariableNode() {variable=Control Name}.Value) {is=struct RigVMUnitNode}
   RigVMVariableNode(Value=@n1.Settings.bVisible, ExecuteContext=@RigVMFunction_Sequence.A) {variable=Visibility}
   RigVMFunction_MathQuaternionToRotator = /Script/RigVM.RigVMFunction_MathQuaternionToRotator(Value=@n1.Settings.Transform.Rotation) {is=struct RigVMUnitNode, template="Cast::Execute(in Value,out Result)"}
   RigVMFunction_MathQuaternionFromRotatorV2 = /Script/RigVM.RigVMFunction_MathQuaternionFromRotatorV2(Value=RigVMDispatchNode(
     Condition=RigVMVariableNode() {variable=Rotation Found}.Value
     True=RigVMVariableNode() {variable=Rotation Result}.Value
     False=@RigVMFunction_MathQuaternionToRotator.Result
    ) {template="DISPATCH_RigVMDispatch_If(in Condition,in True,in False,out Result)"}.Result) {is=struct RigVMUnitNode, template="Cast::Execute(in Value,out Result)"}
   > RigVMVariableNode(Value=@RigVMFunction_MathQuaternionFromRotatorV2.Result) {variable=Rotation}
   > RigVMVariableNode(Value=RigVMDispatchNode(
     False=@n1.Settings.Transform.Translation
     Condition=RigVMVariableNode() {variable=Transform Found}.Value
     True=RigVMVariableNode() {variable=Transform Result}.Value
    ) {template="DISPATCH_RigVMDispatch_If(in Condition,in True,in False,out Result)"}.Result) {variable=Translation}
   DISPATCH_RigVMDispatch_If = RigVMDispatchNode(
    Condition=RigVMVariableNode() {variable=Scale Found}.Value
    True=RigVMVariableNode() {variable=Scale Result}.Value
    False=@n1.Settings.Transform.Scale3D
   ) {template="DISPATCH_RigVMDispatch_If(in Condition,in True,in False,out Result)"}
   > RigVMVariableNode(Value=@DISPATCH_RigVMDispatch_If.Result) {variable=Scale}
  ]}
 New Function_ContainedGraph {nodes [
   RigVMFunctionEntryNode()
   > RigVMFunction_Sequence = RigVMAggregateNode()
   DISPATCH_RigVMDispatch_ArrayIterator = RigVMDispatchNode(ExecuteContext=@RigVMFunction_Sequence.B, Array=RigVMVariableNode() {variable=Children Ctrls}.Value) {template="DISPATCH_RigVMDispatch_ArrayIterator(in Array,out Element,out Index,out Count,out Ratio)"}
   > &RigUnit_SetTransform(
    Weight=1
    bInitial=False
    Item=@DISPATCH_RigVMDispatch_ArrayIterator.Element
    Value=&RigUnit_GetTransform(Item=(Type=Bone,Name="offset"), bInitial=false, Item.Name=@DISPATCH_RigVMDispatch_ArrayIterator.Element.Name).Transform
   )
   RigVMFunctionReturnNode(ExecuteContext=@RigVMFunction_Sequence.C)
   &RigUnit_SetTransform(
    Item=(Type=Control,Name="root_ctrl")
    Weight=1
    bInitial=False
    Value=&RigUnit_GetTransform(Item=(Type=Bone,Name="root"), bInitial=false).Transform
    Item.Name=RigVMVariableNode() {variable=Root Ctrl Name}.Value
    ExecutePin=@RigVMFunction_Sequence.A
   )
   > &RigUnit_SetTransform(
    Item=(Type=Control,Name="offset_ctrl")
    Weight=1
    bInitial=False
    Value=&RigUnit_GetTransform(Item=(Type=Bone,Name="offset"), bInitial=false).Transform
    Item.Name=RigVMVariableNode() {variable=Offset Ctrl Name}.Value
   )
   > &RigUnit_SetTransform(
    Item=(Type=Control,Name="handle_ctrl")
    Weight=1
    bInitial=False
    Value=&RigUnit_GetTransform(Item=(Type=Bone,Name="handle"), bInitial=false).Transform
    Item.Name=RigVMVariableNode() {variable=Handle Bone Name}.Value
   )
  ]}
 RigVMFunctionLibrary {nodes [
   Backwards Solver = RigVMCollapseNode(ExecuteContext=())
   Construct Children Controls = RigVMCollapseNode(ExecuteContext=())
   Construct Gameplay Root Controls = RigVMCollapseNode(ExecuteContext=())
   Forward Controls = RigVMCollapseNode(ExecuteContext=())
   Set Control Shape = RigVMCollapseNode(
    Control Settings=(bVisible=True,Name="Arrow_Thick",Color=(R=1,G=0,B=0,A=1),Transform=(Rotation=(X=-0.5,Y=-0.5,Z=-0.5,W=0.5),Translation=(X=2,Y=0,Z=-2),Scale3D=(X=0.5,Y=0.5,Z=0.5)))
    ExecuteContext=()
   )
   Set Control Shape Color = RigVMCollapseNode(Color=(R=0,G=0,B=0,A=1), ExecuteContext=(), PartialName=attach_)
   Set Shape Settings from Data Asset = RigVMCollapseNode(
    ExecuteContext=()
    New Control Color=(R=0,G=0,B=0,A=0)
    Rotation Result=(Pitch=0,Yaw=0,Roll=0)
    Scale Result=(X=0,Y=0,Z=0)
    Shape Name=Box_Thick
    Transform Result=(X=0,Y=0,Z=0)
   )
  ]}
 RigVMFunction_Sequence_1_ContainedGraph {nodes [
   n1 = RigVMFunctionEntryNode()
   RigVMFunction_Sequence = &RigVMFunction_Sequence(ExecuteContext=@n1.ExecuteContext)
   n2 = &RigVMFunction_Sequence(ExecuteContext=@RigVMFunction_Sequence.B)
   RigVMFunctionReturnNode(A=@RigVMFunction_Sequence.A, B=@n2.A, D=@n2.B)
  ]}
 RigVMFunction_Sequence_1_ContainedGraph {nodes [
   n1 = RigVMFunctionEntryNode()
   RigVMFunction_Sequence = &RigVMFunction_Sequence(ExecuteContext=@n1.ExecuteContext)
   n2 = &RigVMFunction_Sequence(ExecuteContext=@RigVMFunction_Sequence.B)
   n3 = &RigVMFunction_Sequence(ExecuteContext=@n2.B)
   RigVMFunctionReturnNode(A=@RigVMFunction_Sequence.A, B=@n2.A, C=@n3.A, D=@n3.B)
  ]}
 RigVMFunction_Sequence_1_ContainedGraph {nodes [
   n1 = RigVMFunctionEntryNode()
   RigVMFunction_Sequence = &RigVMFunction_Sequence(ExecuteContext=@n1.ExecuteContext)
   n2 = &RigVMFunction_Sequence(ExecuteContext=@RigVMFunction_Sequence.B)
   n3 = &RigVMFunction_Sequence(ExecuteContext=@n2.B)
   n4 = &RigVMFunction_Sequence(ExecuteContext=@n3.B)
   n5 = &RigVMFunction_Sequence(ExecuteContext=@n4.B)
   RigVMFunctionReturnNode(A=@RigVMFunction_Sequence.A, B=@n2.A, C=@n3.A, D=@n4.A, E=@n5.A, F=@n5.B)
  ]}
 RigVMFunction_Sequence_1_ContainedGraph {nodes [
   n1 = RigVMFunctionEntryNode()
   RigVMFunction_Sequence = &RigVMFunction_Sequence(ExecuteContext=@n1.ExecuteContext)
   n2 = &RigVMFunction_Sequence(ExecuteContext=@RigVMFunction_Sequence.B)
   RigVMFunctionReturnNode(A=@RigVMFunction_Sequence.A, B=@n2.A, C=@n2.B)
  ]}
 RigVMFunction_Sequence_1_ContainedGraph {nodes [
   n1 = RigVMFunctionEntryNode()
   RigVMFunction_Sequence = &RigVMFunction_Sequence(ExecuteContext=@n1.ExecuteContext)
   n2 = &RigVMFunction_Sequence(ExecuteContext=@RigVMFunction_Sequence.B)
   RigVMFunctionReturnNode(A=@RigVMFunction_Sequence.A, B=@n2.A, C=@n2.B)
  ]}
 RigVMModel {nodes [
   RigUnit_BeginExecution = &RigUnit_BeginExecution()
   Forward Controls = RigVMFunctionReferenceNode(ExecuteContext=@RigUnit_BeginExecution.ExecutePin)
  ]}
 RigVMModel Backwards Solve Graph {nodes [
   n1 = /Script/ControlRig.RigUnit_InverseExecution() {is=struct RigVMUnitNode}
   Backwards Solver = RigVMFunctionReferenceNode(ExecuteContext=@n1.ExecutePin)
  ]}
 RigVMModel Construction Graph {nodes [
   n1 = &RigUnit_PrepareForExecution()
   Construct Gameplay Root Controls = RigVMFunctionReferenceNode(ExecuteContext=@n1.ExecutePin)
   > Construct Children Controls = RigVMFunctionReferenceNode()
  ]}
]
hierarchy [
 root {loc=<0 0 0>, rot=<0 0 0 1>}
 offset {parent=root, loc=<0 0 0>, rot=<0 0 -0.7071 -0.7071>}
 handle {parent=offset, loc=<0 0 0>, rot=<0 0 0 1>}
 hand_l {parent=handle, loc=<0 0 0>, rot=<0 0 0.7071 -0.7071>}
 hand_r {parent=handle, loc=<0 0 0>, rot=<0 0 0.7071 -0.7071>}
 trigger {parent=handle, loc=<7.5544 0 3.7022>, rot=<0 0 0 1>}
 ammo_compartment {parent=handle, loc=<27.6315 0 7.4206>, rot=<0 0 0 1>}
 bullet {parent=handle, loc=<15.449 0 10.3323>, rot=<0 -0.2164 0 -0.9763>}
 barrel_r_01 {parent=handle, loc=<39.9824 -1.248 12.5287>, rot=<-0.7071 -0.7071 0 0>}
 barrel_r_02 {parent=handle, loc=<46.7413 -0.9847 12.1258>, rot=<-0.7071 -0.7071 0 0>}
 barrel_r_03 {parent=handle, loc=<53.8411 -0.878 11.9041>, rot=<-0.7071 -0.7071 0 0>}
 barrel_l_01 {parent=handle, loc=<39.9824 1.2713 12.5287>, rot=<0 0 0.7071 -0.7071>}
 barrel_l_02 {parent=handle, loc=<46.7413 1.0856 12.1258>, rot=<0 0 0.7071 -0.7071>}
 barrel_l_03 {parent=handle, loc=<53.8411 0.9205 11.9041>, rot=<0 0 0.7071 -0.7071>}
 barrel_r_04 {parent=handle, loc=<61.223 -1.0082 10.8308>, rot=<-0.7071 -0.7071 0 0>}
 barrel_r_05 {parent=handle, loc=<68.097 -1.0126 10.7271>, rot=<-0.7071 -0.7071 0 0>}
 barrel_l_04 {parent=handle, loc=<61.223 1.0082 10.8308>, rot=<0 0 0.7071 -0.7071>}
 barrel_l_05 {parent=handle, loc=<68.097 1.0126 10.7271>, rot=<0 0 0.7071 -0.7071>}
 Shells {parent=root, type=Null, loc=<-0.2187 21.6885 -3.8249>, rot=<0 0 0 1>}
 Muzzle {parent=root, type=Null, loc=<0 82.2436 4.4477>, rot=<0 0 0.7071 0.7071>}
]
variables [
 Children Ctrls {type="TArray<FRigElementKey>"}
 Default Trigger Rot Value {type=FRotator, default=(Pitch=0,Yaw=0,Roll=-90)}
 Default Trigger Scale Value {type=FVector, default=(X=0.2,Y=0.2,Z=0.2)}
 Default Trigger Transform Value {type=FVector, default=(X=1,Y=0,Z=-5.5)}
 Handle Bone Name {type=FName, default=handle}
 Handle Ctrl {type=FRigElementKey, default=(Type=None,Name="")}
 Handle Ctrl Setting {
  type=FRigUnit_HierarchyAddControlTransform_Settings
  default="(InitialSpace=LocalSpace,bUsePreferredRotationOrder=False,PreferredRotationOrder=YZX,Limits=(LimitTranslationX=(bMinimum=False,bMaximum=False),LimitTranslationY=(bMinimum=False,bMaximum=False),LimitTranslationZ=(bMinimum=False,bMaximum=False),LimitPitch=(b ...(truncated)"
 }
 "Left Hand  Ctrl Name" {type=FName, default=hand_l}
 Left Hand Ctrl Settings {
  type=FRigUnit_HierarchyAddControl_ShapeSettings
  default=(bVisible=True,Name="Diamond_Thick",Color=(R=1,G=0.1062,B=0.9795,A=1),Transform=(Rotation=(X=0,Y=0,Z=0,W=1),Translation=(X=0,Y=0,Z=0),Scale3D=(X=0.5,Y=0.5,Z=0.5)))
 }
 Offset Ctrl {type=FRigElementKey, default=(Type=None,Name="")}
 Offset Ctrl Name {type=FName, default=offset}
 Offset Ctrl Setting {
  type=FRigUnit_HierarchyAddControlTransform_Settings
  default="(InitialSpace=LocalSpace,bUsePreferredRotationOrder=False,PreferredRotationOrder=YZX,Limits=(LimitTranslationX=(bMinimum=False,bMaximum=False),LimitTranslationY=(bMinimum=False,bMaximum=False),LimitTranslationZ=(bMinimum=False,bMaximum=False),LimitPitch=(b ...(truncated)"
 }
 Right Hand Ctrl Name {type=FName, default=hand_r}
 Right Hand Ctrl Settings {
  type=FRigUnit_HierarchyAddControl_ShapeSettings
  default=(bVisible=True,Name="Diamond_Thick",Color=(R=0,G=0.021,B=0.4678,A=1),Transform=(Rotation=(X=0,Y=0,Z=0,W=1),Translation=(X=0,Y=0,Z=0),Scale3D=(X=0.45,Y=0.45,Z=0.45)))
 }
 Root Ctrl {type=FRigElementKey, default=(Type=None,Name="")}
 Root Ctrl Name {type=FName, default=root}
 Root Ctrl Setting {
  type=FRigUnit_HierarchyAddControlTransform_Settings
  default="(InitialSpace=LocalSpace,bUsePreferredRotationOrder=False,PreferredRotationOrder=YZX,Limits=(LimitTranslationX=(bMinimum=False,bMaximum=False),LimitTranslationY=(bMinimum=False,bMaximum=False),LimitTranslationZ=(bMinimum=False,bMaximum=False),LimitPitch=(b ...(truncated)"
 }
 Trigger Control Color {type=FLinearColor, default=(R=1,G=0,B=0,A=1)}
 Trigger Ctrl Name {type=FName, default=trigger}
]
