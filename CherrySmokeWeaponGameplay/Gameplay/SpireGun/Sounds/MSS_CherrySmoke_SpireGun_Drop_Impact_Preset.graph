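# MetaSound Source preset that plays the landing impact when the CherrySmoke SpireGun item is dropped.
# It inherits its graph from MSS_Item_Drop_Impact and overrides inputs with three landing wave variations, a reduced master volume and a narrow random pitch-shift spread.
# Output is a mono one-shot using shared Fortnite item attenuation and concurrency, routed to the Fort_Loot_Collect sound class.
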MetaSoundSource /CherrySmokeWeaponGameplay/Gameplay/SpireGun/Sounds/MSS_CherrySmoke_SpireGun_Drop_Impact_Preset
preset_parent=/FortMetasound/Items/MSS_Item_Drop_Impact
inputs [
 MasterVolume {type=Float}
 Random.PitchShift.Base {type=Float}
 Random.PitchShift.Spread {type=Float}
 Sounds {type=WaveAsset:Array}
 UE.Source.OnPlay {type=Trigger}
]
interfaces [UE.OutputFormat.Mono:1.0, UE.Source.OneShot:1.0, UE.Source:1.1]
outputs [
 "UE.OutputFormat.Mono.Audio:0" {type=Audio}
 UE.Source.OneShot.OnFinished {type=Trigger}
]
pages [Default {input_defaults {
   MasterVolume=0.4
   Random.PitchShift.Base=0
   Random.PitchShift.Spread=0.5
   UE.Source.OnPlay=null
   Sounds [
    /CherrySmokeWeaponGameplay/Gameplay/SpireGun/Sounds/wavs/CherrySmoke_SpireGun_Drop_Land_01
    /CherrySmokeWeaponGameplay/Gameplay/SpireGun/Sounds/wavs/CherrySmoke_SpireGun_Drop_Land_02
    /CherrySmokeWeaponGameplay/Gameplay/SpireGun/Sounds/wavs/CherrySmoke_SpireGun_Drop_Land_03
   ]
  }}]
properties {
 assetImportData=&self:AssetImportData
 attenuationSettings=/FortMetasound/Items/AS_Fort_Item
 maxDistance=3500
 sampleRate=48000
 soundClassObject=&Fort_Loot_Collect
 concurrencySet [&CS_Fort_Item]
}
